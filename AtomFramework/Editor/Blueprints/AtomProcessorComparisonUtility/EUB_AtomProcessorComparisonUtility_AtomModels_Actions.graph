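# Editor utility blueprint in AtomFramework's processor comparison tools, offering an action to compare Fast Bricks triangle counts for Atom models.
# CompareFastBricksTriCounts gets the editor world from the Unreal Editor Subsystem and passes it to GetModelFastBricksTriCounts.
# The returned per-model tri count list is left unconnected, so the blueprint itself does no further comparison or reporting; the event graph is empty.
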EditorUtilityBlueprint /AtomFramework/Editor/Blueprints/AtomProcessorComparisonUtility/EUB_AtomProcessorComparisonUtility_AtomModels_Actions
graphs [
 CompareFastBricksTriCounts {nodes [
   CompareFastBricksTriCounts() {is=title K2Node_FunctionEntry}
   > n1 = GetEditorWorld(self=Get UnrealEditorSubsystem())
   > GetModelFastBricksTriCounts!(execute, __WorldContext:Object Reference=@n1) -> (then, Model Tri Counts List:Array of STRUCT Atom Fast Bricks Model Tri Count Structures)
  ]}
 EventGraph {links [], nodes []}
]
variables []
