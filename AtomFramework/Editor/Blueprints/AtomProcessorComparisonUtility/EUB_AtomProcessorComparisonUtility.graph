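# Editor utility actor that lays out Atom model static meshes for visual comparison, labelled with LOD, triangle count, material count and processor.
# Construction script and events rebuild spawned components as either a line of filtered models or per-model columns of LODs with a transition preview.
# LOD view labels add FastBricks counts; over-budget triangle counts can recolor labels and appear in a summary.
# Events control view, rotation, forced LOD and Nanite.
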EditorUtilityBlueprint /AtomFramework/Editor/Blueprints/AtomProcessorComparisonUtility/EUB_AtomProcessorComparisonUtility
graphs [
 AddComponentInfoDisplay {nodes [
   n1 = AddComponentInfoDisplay!() -> (
    then
    LOD Index:int
    AssetIndex:int
    Location:Vector=~
    LocalOffset:Vector=~
    AdditionalString:String
    OverrideTextColor:bool=~
    OverrideColor:Linear Color Structure
    IsAutoLOD:bool=~
    IsLodView:bool=~
   ) {is=title K2Node_FunctionEntry}
   > n2 = Add Text Render Component!(
    execute:Exec
    self:Actor Object Reference
    bManualAttachment:bool=true
    RelativeTransform_Location:Vector=@n1.Location
    RelativeTransform_Rotation:Rotator="90,0,0"
    RelativeTransform_Scale:Vector="1,1,1"
    ComponentTemplateContext:Object Reference
   ) -> (then:Exec, ReturnValue:Text Render Component Object Reference) {is=title K2Node_AddComponent}
   > n3 = Branch(Condition=@n1.OverrideTextColor)
   > n4 = SetTextRenderColor!(
    execute
    self:Text Render Component Object Reference=@n2
    Value:Color Structure="To Color (LinearColor)"!(InLinearColor:Linear Color Structure=@n1.OverrideColor, InUseSRGB:bool=true) -> (ReturnValue:Color Structure)
   ) -> (then)
   Attach Component To Component!(
    execute=[@n5.then, @n4.then]
    self:Scene Component Object Reference=@n2
    Parent:Scene Component Object Reference
    SocketName:Name=None
    LocationRule:EAttachmentRule Enum=KeepWorld
    RotationRule:EAttachmentRule Enum=KeepWorld
    ScaleRule:EAttachmentRule Enum=KeepWorld
    bWeldSimulatedBodies:bool=false
   ) -> (then, ReturnValue:bool)
   > SetWorldSize!(execute, self:Text Render Component Object Reference=@n2, Value:float=Get FontSize() -> (FontSize:double)) -> (then)
   > SetHorizontalAlignment!(execute, self:Text Render Component Object Reference=@n2, Value:EHorizTextAligment Enum=EHTA_Center) -> (then)
   > Add(
    TargetArray:Array of Actor Component Object References=Get ChildComponents() -> (ChildComponents:Array of Actor Component Object References)
    NewItem:"Actor Component Object Reference (by ref)"=@n2
   ) {is=title K2Node_CallArrayFunction}
   > SetComponentDisplayInfo!(
    execute
    self
    TextRenderComponent:Text Render Component Object Reference=@n2
    LOD Index:int=@n1
    AssetIndex:int=@n1.AssetIndex
    AdditionalString:String=@n1.AdditionalString
    OverrideColor:bool=@n1.OverrideTextColor
    IsAutoLOD:bool=Get IsAutoLOD() -> (IsAutoLOD:bool)
    IsLodView:bool=Get IsLodView() -> (IsLodView:bool)
   ) -> (then)
   > Add Local Offset!(execute, self:Scene Component Object Reference=@n2, DeltaLocation:Vector=@n1.LocalOffset, bSweep:bool=false, bTeleport:bool=false) -> (then, SweepHitResult:Hit Result Structure)
   > Return Node!(execute, TextRenderComponent:Text Render Component Object Reference=@n2)
   n5 = SetTextRenderColor(execute=@n3.else, self=@n2, Value=Get TextColor() -> (TextColor:Color Structure))
  ]}
 AttachAdditionalParts {nodes [
   n1 = AttachAdditionalParts!() -> (
    then
    AdditionalMeshes:Array of Static Mesh Soft Object References
    AtomStaticMeshComponent:Static Mesh Component Object Reference
    LODModel:int
   ) {is=title K2Node_FunctionEntry}
   > n2 = For Each Loop(Array:Array of Static Mesh Soft Object References=@n1) -> (Array Element:Static Mesh Soft Object Reference)
   Return Node!(execute=[@n2.Completed, @n4.Is Not Valid])
   n3 = Resolve Soft Reference!(Input:Static Mesh Soft Object Reference=@n2) -> (Output:Static Mesh Object Reference)
   > n4 = Is Valid(InputObject=@n3)
   > n5 = Add Static Mesh Component!(
    execute:Exec
    self:Actor Object Reference
    bManualAttachment:bool=true
    RelativeTransform:Transform
    ComponentTemplateContext:Object Reference
   ) -> (then:Exec, ReturnValue:Static Mesh Component Object Reference) {is=title K2Node_AddComponent}
   > Attach Component To Component(
    self=@n5
    Parent=@n1.AtomStaticMeshComponent
    SocketName=None
    LocationRule=KeepRelative
    RotationRule=KeepRelative
    ScaleRule=KeepRelative
    bWeldSimulatedBodies=TRUE
   )
   > SetStaticMesh(self=@n5, NewMesh=@n3)
   > n6 = Branch(Condition=Get GridView() -> (GridView:bool))
   Add(execute=[@n7.then, @n6.then], TargetArray=Get ChildComponents(), NewItem=@n5)
   n7 = SetForcedLodModel!(execute=@n6.else, self:Static Mesh Component Object Reference=@n5, NewForcedLodModel:int=@n1.LODModel) -> (then)
  ]}
 CenterSceneComponent {nodes [
   n1 = CenterSceneComponent!() -> (then, StaticMeshComponent:Static Mesh Component Object Reference) {is=title K2Node_FunctionEntry}
   n2 = GetLocalBounds!(StaticMeshComponent:Scene Component Object Reference=@n1) -> (Origin:Vector, BoxExtent:Vector, SphereRadius:float)
   n3 = BreakVector(InVec=@n2.BoxExtent)
   n4 = vector - vector!(A=Get RelativeLocation!(self:Scene Component Object Reference=@n1) -> (RelativeLocation:Vector), B=@n2) -> (ReturnValue_X:double, ReturnValue_Y:double, ReturnValue_Z:double)
   Set Relative Location!(
    execute
    self:Scene Component Object Reference=@n1
    NewLocation:Vector=MakeVector(X=@n4, Y=@n4.ReturnValue_Y, Z=float + float(A=@n4.ReturnValue_Z, B=@n3.Z))
    bSweep:bool=false
    bTeleport:bool=false
   ) -> (then, SweepHitResult:Hit Result Structure) {is=title K2Node_CallFunction}
   > Return Node!(execute, UpdatedTransform:Transform=GetRelativeTransform!(self:Scene Component Object Reference=@n1) -> (ReturnValue:Transform))
   Get RelativeLocation(self=@n1)
  ]}
 ComposeAssetInfo {nodes [
   n1 = ComposeAssetInfo!() -> (
    then
    TextRenderComp:Text Render Component Object Reference
    AdditionalString:String
    AssetIndex:int
    AssetName:String
    NumMaterials:int
    LODIndex:int
    AssetTriCount:int
    SphereRadius:int
    IsAutoLOD:bool=~
    AssetProcessor:String
   ) {is=title K2Node_FunctionEntry}
   n2 = Append!(A=@n1.AdditionalString, B="To String (Integer)"(InInt=@n1.AssetIndex), C:String=" | ") -> (ReturnValue)
   n3 = Append!(A=@n1.AssetName, B="\r\n") -> (ReturnValue)
   n4 = Append!(A="Number of Materials: ", B="To String (Integer)"(InInt=@n1.NumMaterials), C:String="\r\n") -> (ReturnValue)
   n5 = Append!(A="Tri Count: ", B="To String (Integer)"(InInt=@n1.AssetTriCount), C="\r\n", D:String="Processor: ", E:String=@n1.AssetProcessor) -> (ReturnValue)
   n6 = Select(
    Option 0:String=Append!(A=LOD, B="To String (Integer)"(InInt=@n1.LODIndex), C=" | ") -> (ReturnValue)
    Option 1:String=Append(A="LOD ", B=Auto, C=" | ")
    Index=Get IsAutoLOD()
   ) -> (ReturnValue:String)
   > Set Text!(
    execute
    self:Text Render Component Object Reference=@n1
    Value:"Text (by ref)"="To Text (String)"(InString=Append!(A=@n2, B=@n3, C=@n4, D:String=@n6, E:String=@n5, F:String) -> (ReturnValue))
   ) -> (then) {is=title K2Node_CallFunction}
   > Return Node!(execute)
  ]}
 CountFastBricks {nodes [
   n1 = CountFastBricks!() -> (then, Model:Atom Model Object Reference, LOD:int) {is=title K2Node_FunctionEntry}
   > n2 = For Each Loop(Array:Array of Atom Model Primitive Structures=Get Primitives!(self:Atom Model Object Reference=@n1) -> (Primitives:Array of Atom Model Primitive Structures)) -> (Array Element:Atom Model Primitive Structure)
   n3 = Get FastBricks() -> (FastBricks:int=0)
   n4 = Get TotalBricks() -> (TotalBricks:int=0)
   Return Node!(
    execute=@n2.Completed
    CountString:String=Append(A="\r\n", B="FastBricks: ", C="To String (Integer)"(InInt=@n3), D=" / ", E="To String (Integer)"(InInt=@n4), F="\r\n")
   )
   > n5 = For Each Loop(Array:Array of Atom Model Part Structures=Break Atom Model Primitive!(AtomModelPrimitive:"Atom Model Primitive Structure (by ref)"=@n2) -> (Parts:Array of Atom Model Part Structures, DesignId:int, DesignName:Name) {is=title K2Node_BreakStruct}) -> (Array Element:Atom Model Part Structure)
   > Increment Int(Value=Get TotalBricks() -> (TotalBricks=0))
   n6 = Resolve Soft Reference(Input:Atom Primitive Soft Object Reference=Break Atom Model Part!(AtomModelPart:"Atom Model Part Structure (by ref)"=@n5) -> (AtomPrimitive:Atom Primitive Soft Object Reference, Transforms:Array of Transforms, bIgnoreCommonPartCulling:bool=~) {is=title K2Node_BreakStruct}) -> (Output:Atom Primitive Object Reference)
   > Branch(Condition=IsValid(Object=GetGeometryContainerForExportStyle!(
      self:Atom Primitive Object Reference=@n6
      ExportStyleName:String=Game
      FallbackExportStyleName:String
      LODIndex:int=@n1.LOD
      bUseFallbacks:bool=true
     ) -> (ReturnValue:Atom Primitive Geometry Container Object Reference)))
   > Increment Int(Value=Get FastBricks() -> (FastBricks=0))
  ]}
 EventGraph {nodes [
   UpdateInfo() {is=title K2Node_CustomEvent}
   > n1 = Branch(Condition=Get GridView())
   > SetupComponent_LineView()
   SetupComponent_LodView(execute=@n1.else)
   ModelNext() {is=title K2Node_CustomEvent}
   n2 = Get CurrentIndex() -> (CurrentIndex:int=0)
   > n3 = Increment Int(Value=@n2)
   n4 = Get CurrentIndex() -> (CurrentIndex=0)
   n5 = Branch(execute=[@n3."  ", @n17."  "], Condition="integer > integer"(A=@n4, B=Get LastModelIndex() -> (LastModelIndex:int=0)))
   > n6 = Set CurrentIndex(CurrentIndex:int=0) -> (Output_Get:int=0)
   n7 = Branch(execute=[@n6.then, @n9.then, @n8.else], Condition=Get GridView())
   > Set AtomModelToDisplay(AtomModelToDisplay:int="Get (a ref)"(Array=Get FilteredMeshIDs() -> (FilteredMeshIDs:Array of Integers), Dimension 1=Get CurrentIndex() -> (CurrentIndex=0))) -> (Output_Get:int=0)
   Set AtomModelToDisplay(execute=@n7.else, AtomModelToDisplay="Get (a ref)"(Array=Get FilteredMeshIDs(), Dimension 1=Get CurrentIndex() -> (CurrentIndex=0))) -> (Output_Get=0)
   > UpdateInfo() {is=title K2Node_CallFunction}
   n8 = Branch(execute=@n5.else, Condition="integer < integer"(A=@n4))
   > n9 = Set CurrentIndex(CurrentIndex=Get LastModelIndex() -> (LastModelIndex=0)) -> (Output_Get=0)
   - LOD ViewMode -() {is=title K2Node_CustomEvent}
   > n10 = Set GridView(GridView:bool=false) -> (Output_Get:bool)
   UpdateInfo(execute=[@n12.then, @n10.then])
   RotateAntiClock() {is=title K2Node_CustomEvent}
   > n11 = Set YawCustomRotation(YawCustomRotation:double=float - float(A=Get YawCustomRotation() -> (YawCustomRotation:double), B=Get RotationIncrement() -> (RotationIncrement:double))) -> (Output_Get:double)
   UpdateInfo(execute=[@n22.then, @n23.then, @n11.then])
   "- GRID ViewMode - "() {is=title K2Node_CustomEvent}
   > n12 = Set GridView(GridView=true)
   LOD Bias Previous() {is=title K2Node_CustomEvent}
   > ExecuteConsoleCommand(Command=r.Nanite 0)
   n13 = Get LODLevel() -> (LODLevel:int=0)
   > n14 = Branch(Condition="integer > integer"(A=@n13))
   > n15 = Decrement Int(Value=@n13)
   UpdateInfo(execute=[@n20.else, @n21."  ", @n15."  ", @n18.then, @n14.else, @n16.then])
   Enable Nanite() {is=title K2Node_CustomEvent}
   > ExecuteConsoleCommand(Command=r.Nanite 1)
   > n16 = Set LODLevel(LODLevel:int=0) -> (Output_Get:int=0)
   ModelPrevious() {is=title K2Node_CustomEvent}
   > n17 = Decrement Int(Value=@n2)
   Disable Nanite/LOD Auto() {is=title K2Node_CustomEvent}
   > ExecuteConsoleCommand(Command=r.Nanite 0)
   > n18 = Set LODLevel(LODLevel=0) -> (Output_Get=0)
   LOD Bias Next() {is=title K2Node_CustomEvent}
   > ExecuteConsoleCommand(Command=r.Nanite 0)
   n19 = Get LODLevel() -> (LODLevel=0)
   > n20 = Branch(Condition="integer < integer"(A=@n19, B=Get NumLODs() -> (NumLODs:int=0)))
   > n21 = Increment Int(Value=@n19)
   RotationReset() {is=title K2Node_CustomEvent}
   > n22 = Set YawCustomRotation(YawCustomRotation=0)
   RotateClock() {is=title K2Node_CustomEvent}
   > n23 = Set YawCustomRotation(YawCustomRotation=float + float(A=Get YawCustomRotation(), B=Get RotationIncrement()))
   Event BeginPlay()
   > UpdateInfo()
  ]}
 GetLocalBounds {nodes [
   n1 = GetLocalBounds!() -> (then, StaticMeshComponent:Scene Component Object Reference) {is=title K2Node_FunctionEntry}
   n2 = GetComponentBounds!(Component:Scene Component Object Reference=@n1) -> (Origin:Vector, BoxExtent:Vector, SphereRadius:float)
   > Return Node!(
    execute
    Origin:Vector=InverseTransformLocation!(
     T:"Transform (by ref)"=Get World Transform!(self:Scene Component Object Reference=@n1) -> (ReturnValue:Transform) {is=title K2Node_CallFunction}
     Location:Vector=@n2
    ) -> (ReturnValue:Vector)
    BoxExtent:Vector=@n2.BoxExtent
    SphereRadius:float=@n2.SphereRadius
   )
  ]}
 GridLoop {nodes [
   n1 = Inputs!() -> (NumMeshes:int)
   n2 = Ceil(A=Sqrt!(A:double="To Float (Integer)"(InInt=@n1)) -> (ReturnValue:double)) {is=title K2Node_CallFunction}
   For Loop(LastIndex=@n2)
   > n3 = For Loop(LastIndex=@n2)
   Set CurrentPosX(execute=@n3.Completed, CurrentPosX:double=float + float(A=Get CurrentPosX() -> (CurrentPosX:double), B=Get BaseSize() -> (BaseSize:double))) -> (Output_Get:double)
   > Set CurrentPosY(CurrentPosY:double=Get InitialPosY() -> (InitialPosY:double)) -> (Output_Get:double)
   n4 = Get CurrentPosX()
   n5 = Get InitialPosZ() -> (InitialPosZ:double)
   n6 = Get CurrentPosY() -> (CurrentPosY:double)
   Outputs!(Transform:Transform=MakeTransform!(Location=MakeVector(X=@n4, Y=@n6, Z=@n5), Rotation_Roll:float=0, Rotation_Pitch:float=0, Rotation_Yaw:float=0, Scale) -> (ReturnValue))
  ]}
 InitializeCommon {nodes [
   InitializeCommon() {is=title K2Node_FunctionEntry}
   > n1 = Sequence!(execute) -> (then_0, then_1, then_2:Exec, then_3:Exec)
   Set Text(execute=@n1.then_1, self=Get InfoDisplay() -> (InfoDisplay:Text Render Component Object Reference), Value=MakeLiteralText())
   Set AtomSelection(
    execute=@n1.then_2
    AtomSelection:bool=Is Not Empty!(TargetArray:Array of STRUCT Atom Processor Comparison Utility Structures=Get AtomMeshes() -> (AtomMeshes:Array of STRUCT Atom Processor Comparison Utility Structures)) -> (ReturnValue:bool)
   ) -> (Output_Get:bool)
   n2 = Branch(execute=@n1.then_3, Condition=Get AtomSelection() -> (AtomSelection:bool))
   n3 = Length(TargetArray:Array of Integers=Get FilteredMeshIDs())
   > Set NumModels(NumModels:int=@n3) -> (Output_Get:int=0)
   > Set LastModelIndex(LastModelIndex:int=int - int(A=@n3)) -> (Output_Get:int=0)
   > Set AtomModelToDisplay(AtomModelToDisplay="Get (a ref)"(Array=Get FilteredMeshIDs(), Dimension 1=Get CurrentIndex() -> (CurrentIndex=0))) -> (Output_Get=0)
   > n4 = Sequence!(execute) -> (then_0, then_1)
   > n5 = For Each Loop(Array:Array of STRUCT Atom Processor Comparison Utility Structures=Get AtomMeshes()) -> (Array Element:STRUCT Atom Processor Comparison Utility Structure)
   n6 = GetNumLODs(self=Resolve Soft Reference(Input:Static Mesh Soft Object Reference=Break STRUCT Atom Processor Comparison Utility!(STRUCT_AtomProcessorComparisonUtility:"STRUCT Atom Processor Comparison Utility Structure (by ref)"=@n5) -> (
      AtomModel_35_#1:Atom Model Soft Object Reference
      AtomMeshMain_36_#2:Static Mesh Soft Object Reference
      AtomMeshAdditional_38_#3:Array of Static Mesh Soft Object References
     ) {is=title K2Node_BreakStruct}.AtomMeshMain_36_#2) -> (Output:Static Mesh Object Reference))
   > Branch(Condition="integer > integer"(A=@n6, B=Get NumLODs() -> (NumLODs=0)))
   > Set NumLODs(NumLODs:int=@n6) -> (Output_Get:int=0)
   Return Node!(execute=[@n7.then, @n4.then_1], AtomSelectionValid:bool=Get AtomSelection())
   n7 = Set AtomModelToDisplay(execute=@n2.else, AtomModelToDisplay=0) -> (Output_Get=0)
   n8 = Get AtomModelToDisplay() -> (AtomModelToDisplay:int=0)
   n9 = Get FilteredMeshIDs()
   n10 = Branch(
    execute=@n1.then_0
    Condition=AND Boolean(A=Is Not Empty(TargetArray:Array of Integers=@n9), B=Contains Item(TargetArray:Array of Integers=@n9, ItemToFind:"Integer (by ref)"=@n8))
   )
   > n11 = Set CurrentIndex(CurrentIndex=Find Item!(TargetArray:Array of Integers=@n9, ItemToFind:"Integer (by ref)"=@n8) -> (ReturnValue)) -> (Output_Get=0)
   Set AtomModelToDisplay(execute=[@n12.then, @n11.then], AtomModelToDisplay="Get (a ref)"(Array=@n9, Dimension 1=Get CurrentIndex() -> (CurrentIndex=0))) -> (Output_Get=0)
   n12 = Set CurrentIndex(execute=@n10.else, CurrentIndex=0) -> (Output_Get=0)
  ]}
 InitializeValues_GridView {nodes [
   InitializeValues_GridView() {is=title K2Node_FunctionEntry}
   > n1 = InitializeCommon!(execute, self) -> (then, AtomSelectionValid:bool) {is=title K2Node_CallFunction}
   > n2 = Sequence!(execute) -> (then_0, then_1, then_2:Exec, then_3:Exec, then_4:Exec)
   > Set ControlID(ControlID:int=0) -> (Output_Get:int=0)
   n3 = Get ChildComponents()
   > Branch(Condition=Is Not Empty(TargetArray:Array of Actor Component Object References=@n3))
   > n4 = For Each Loop(Array:Array of Actor Component Object References=@n3) -> (Array Element:Actor Component Object Reference)
   Clear(execute=@n4.Completed, TargetArray:Array of Actor Component Object References=Get ChildComponents())
   Destroy Component!(execute=@n4.LoopBody, self:Actor Component Object Reference=@n4, Object:Object Reference) -> (then)
   n5 = Set InitialPosX(execute=@n2.then_1, InitialPosX:double=0) -> (Output_Get:double)
   > Set CurrentPosX(CurrentPosX=@n5)
   n6 = Set InitialPosY(execute=@n2.then_2, InitialPosY:double=0) -> (Output_Get:double)
   > Set CurrentPosY(CurrentPosY=@n6)
   n7 = Set InitialPosZ(execute=@n2.then_3, InitialPosZ:double=Get HeightOffset() -> (HeightOffset:double)) -> (Output_Get:double)
   > Set CurrentPosZ(CurrentPosZ:double=@n7) -> (Output_Get:double)
   n8 = "Max (Float)"(A=Get LargestBound() -> (LargestBound:double), B=170)
   Set BaseSize(
    execute=@n2.then_4
    BaseSize:double=float + float(A=float + float(A=@n8, B=30), B=Get AdditionalSpacingOnGrid() -> (AdditionalSpacingOnGrid:double))
   ) -> (Output_Get:double)
   > Return Node!(execute, AtomSelectionValid?:bool=@n1)
  ]}
 InitializeValues_LodView {nodes [
   InitializeValues_LodView() {is=title K2Node_FunctionEntry}
   > n1 = InitializeCommon()
   > Set CurrentOffsetSingle(CurrentOffsetSingle:double=0) -> (Output_Get:double)
   n2 = Get ChildComponents()
   > n3 = Branch(Condition=Is Not Empty(TargetArray=@n2))
   > n4 = For Each Loop(Array=@n2)
   n5 = Clear(execute=@n4.Completed, TargetArray=Get ChildComponents())
   Return Node(execute=[@n5.then, @n3.else], AtomSelectionValid?=@n1)
   Destroy Component(execute=@n4.LoopBody, self=@n4)
  ]}
 SetComponentDisplayInfo {nodes [
   n1 = SetComponentDisplayInfo!() -> (
    then
    TextRenderComponent:Text Render Component Object Reference
    LOD Index:int
    AssetIndex:int
    AdditionalString:String
    OverrideColor:bool=~
    IsAutoLOD:bool=~
    IsLodView:bool=~
   ) {is=title K2Node_FunctionEntry}
   > n2 = Branch(Condition=Get IsLodView())
   > n3 = Set LODLevelDisplay(LODLevelDisplay:int=Get LOD Index() -> (LOD Index:int=0)) -> (Output_Get:int=0)
   n4 = Sequence!(execute=[@n3.then, @n25.then]) -> (then_0, then_1, then_2)
   n5 = Break STRUCT Atom Processor Comparison Utility(STRUCT_AtomProcessorComparisonUtility="Get (a ref)"(Array:Array of STRUCT Atom Processor Comparison Utility Structures=Get AtomMeshes(), Dimension 1=@n1.AssetIndex) -> (Output:"STRUCT Atom Processor Comparison Utility Structure (by ref)"))
   > n6 = LoadAsset_Blocking(Asset=@n5.AtomMeshMain_36_#2)
   > n7 = Cast To StaticMesh(Object=@n6)
   > Set MeshTriCount(MeshTriCount:int=GetNumTriangles!(self:Static Mesh Object Reference=@n7, LODIndex:int=Get LODLevelDisplay() -> (LODLevelDisplay:int=0)) -> (ReturnValue:int=0)) -> (Output_Get:int=0)
   n8 = Set Text(execute=@n7.CastFailed, self=@n1, Value="To Text (String)"(InString=Empty Selection))
   Return Node!(execute=[@n20.then, @n8.then, @n19.then, @n18.else])
   n9 = Resolve Soft Reference(Input:Atom Model Soft Object Reference=@n5) -> (Output:Atom Model Object Reference)
   n10 = GetModelProcessor!(
    execute=@n4.then_2
    ProcessorInstance:"Atom Model Processor Instance Structure (by ref)"="Get (a copy)"(Array:Array of Atom Model Processor Instance Structures=Get ModelProcessorInstances!(self:Atom Model Object Reference=@n9) -> (ModelProcessorInstances:Array of Atom Model Processor Instance Structures)) -> (Output:Atom Model Processor Instance Structure)
   ) -> (then, ReturnValue:Atom Model Processor Object Reference)
   n11 = GetDisplayName(Object=@n7)
   n12 = Split(SourceString=@n11, InStr=SM_, SearchDir=FromStart)
   n13 = GetBounds!(self:Static Mesh Object Reference=@n7) -> (ReturnValue:Box Sphere Bounds Structure)
   > ComposeAssetInfo!(
    execute
    self
    TextRenderComp:Text Render Component Object Reference=@n1
    AdditionalString:String=@n1.AdditionalString
    AssetIndex:int=@n1.AssetIndex
    AssetName:String=Select(Option 0=@n11, Option 1=@n12.RightS, Index=@n12.ReturnValue)
    NumMaterials:int=GetNumSections!(self:Static Mesh Object Reference=@n7, InLOD:int=Get LODLevelDisplay() -> (LODLevelDisplay=0)) -> (ReturnValue:int=0)
    LODIndex:int=Get LODLevelDisplay() -> (LODLevelDisplay=0)
    AssetTriCount:int=Get MeshTriCount() -> (MeshTriCount:int=0)
    SphereRadius:int=Round(A=BreakBoxSphereBounds!(InBoxSphereBounds:"Box Sphere Bounds Structure (by ref)"=@n13) -> (Origin:Vector, BoxExtent:Vector, SphereRadius:float).SphereRadius) {is=title K2Node_CallFunction}
    IsAutoLOD:bool=Get IsAutoLOD()
    AssetProcessor:String=GetObjectName(Object=@n10)
   ) -> (then) {is=title K2Node_CallFunction}
   n14 = Get Budget() -> (Budget:int=0)
   n15 = Get MeshTriCount() -> (MeshTriCount=0)
   n16 = Get LODLevelDisplay() -> (LODLevelDisplay=0)
   n17 = Get BudgetedLOD() -> (BudgetedLOD:int=0)
   > n18 = Branch(Condition=AND Boolean(A="integer > integer"(A=@n15, B=@n14), B="integer >= integer"(A=@n16, B=@n17)))
   > n19 = Branch(Condition=Get OverrideColor() -> (OverrideColor:bool))
   n20 = SetTextRenderColor(execute=@n19.else, self=@n1, Value=Get OverBudgetedColor() -> (OverBudgetedColor:Color Structure))
   n21 = For Each Loop(Exec=@n4.then_1, Array:Array of Static Mesh Soft Object References=@n5.AtomMeshAdditional_38_#3) -> (Array Element:Static Mesh Soft Object Reference)
   > n22 = LoadAsset_Blocking(Asset=@n21)
   > n23 = Cast To StaticMesh(Object=@n22)
   > Is Valid(InputObject=@n23)
   n24 = Get LODLevelDisplay() -> (LODLevelDisplay=0)
   > Set MeshTriCount(MeshTriCount=int + int(A=Get MeshTriCount() -> (MeshTriCount=0), B=GetNumTriangles(self=@n23, LODIndex=@n24) -> (ReturnValue=0))) -> (Output_Get=0)
   n25 = Set LODLevelDisplay(execute=@n2.else, LODLevelDisplay="Max (Integer)"(A=int - int(A=@n1.LOD Index), B=0)) -> (Output_Get=0)
  ]}
 SetGeneralDisplayInfo {nodes [
   n1 = SetGeneralDisplayInfo!() -> (then, AtomSelectionValid:bool=~, DisplayPositionOffset:double=~) {is=title K2Node_FunctionEntry}
   > Set IsAutoLOD(IsAutoLOD:bool="Equal (Integer)"(A=Get LODLevel() -> (LODLevel=0))) -> (Output_Get:bool)
   n2 = Get LODLevel() -> (LODLevel=0)
   > Set LODLevelDisplay(LODLevelDisplay="Max (Integer)"(A=int - int(A=@n2), B=0)) -> (Output_Get=0)
   > SetTextRenderColor(self=Get InfoDisplay(), Value=Get TextColor())
   > SetWorldSize(self=Get InfoDisplay(), Value=Get GeneralDisplayFontSize() -> (GeneralDisplayFontSize:double))
   > n3 = Branch(Condition=@n1)
   n4 = Get Budget() -> (Budget=0)
   n5 = Get BudgetedLOD() -> (BudgetedLOD=0)
   n6 = Length(TargetArray=Get OverBudgetedIDs() -> (OverBudgetedIDs:Array of Integers))
   n7 = StringCombine!(
    Base String:String=Append!(
     A="\r\n\r\nBudget(LOD"
     B="To String (Integer)"(InInt=@n5)
     C="): "
     D="To String (Integer)"(InInt=@n4)
     E="\r\nOver-Budgeted Models: "
     F="To String (Integer)"(InInt=@n6)
     G:String="\r\n\r\n"
    ) -> (ReturnValue)
    IDs:Array of Integers=Get OverBudgetedIDs()
    Names:Array of Strings=Get HighestTriangleNames() -> (HighestTriangleNames:Array of Strings)
    TriangleCounts:Array of Integers=Get HighestTriangleCounts() -> (HighestTriangleCounts:Array of Integers)
   ) -> (CombinedString:String)
   n8 = Break Int Vector!(IntVector:"Int Vector Structure (by ref)"=Get RadiusInfo() -> (RadiusInfo:Int Vector Structure)) -> (X:int, Y:int, Z:int) {is=title K2Node_BreakStruct}
   n9 = Append!(
    A="\r\nSphereRadius: \r\n    Average:  "
    B="To String (Integer)"(InInt=@n8)
    C="\r\n    Largest:  "
    D="To String (Integer)"(InInt=@n8.Y)
    E="\r\n    Smalllest:  "
    F="To String (Integer)"(InInt=@n8.Z)
   ) -> (ReturnValue)
   n10 = Append!(
    A=Get LoadedPath() -> (LoadedPath:String)
    B="\r\nNumber of valid Assets: "
    C="To String (Integer)"(InInt=Get NumModels() -> (NumModels:int=0))
   ) -> (ReturnValue)
   n11 = Get LODLevelDisplay() -> (LODLevelDisplay=0)
   n12 = Append!(A="\r\nDisplaying LOD", B=Select(Option 0="To String (Integer)"(InInt=@n11), Option 1=" Auto", Index=Get IsAutoLOD())) -> (ReturnValue)
   n13 = Append(A="\r\nAverage Triangle Count: ", B="To String (Integer)"(InInt=Get AverageTriangleCount() -> (AverageTriangleCount:int=0)))
   > Set Text(
    self=Get InfoDisplay()
    Value="To Text (String)"(InString=Append!(A=@n10, B=@n12, C:String=@n13, D:String=@n9, E:String=@n7) -> (ReturnValue))
   )
   > Set TempSceneComponents(TempSceneComponents:Array of Scene Component Object References=GetChildrenComponents!(self:Scene Component Object Reference=Get Scene() -> (Scene:Scene Component Object Reference), bIncludeAllDescendants:bool=true) -> (Children:Array of Scene Component Object References)) -> (Output_Get:Array of Scene Component Object References)
   n14 = Get InfoDisplay()
   n15 = BreakVector(InVec=GetComponentBounds(Component=@n14).BoxExtent)
   > Set Relative Location!(
    execute
    self=@n14
    NewLocation_X:double=0
    NewLocation_Y:double=float + float(A=float * float(A=@n15.Y, B=2) {is=title K2Node_PromotableOperator}, B=@n1.DisplayPositionOffset)
    NewLocation_Z:double=0
    bSweep=false
    bTeleport=false
   ) -> (then, SweepHitResult)
   > Return Node()
   n16 = Append!(A=Valid Atom Models not found, B) -> (ReturnValue)
   PrintString(execute=@n3.else, InString=@n16, Duration=5)
   > Set Text(self=Get InfoDisplay(), Value="To Text (String)"(InString=@n16))
  ]}
 SetupComponent_LineView {nodes [
   SetupComponent_LineView() {is=title K2Node_FunctionEntry}
   > n1 = InitializeValues_GridView!(execute, self) -> (then, AtomSelectionValid?:bool) {is=title K2Node_CallFunction}
   > n2 = Branch(Condition=@n1)
   > For Loop(LastIndex=Get NumModels() -> (NumModels=0))
   > n3 = Branch(Condition="integer <= integer"(A=Get ControlID() -> (ControlID:int=0), B=Get LastModelIndex() -> (LastModelIndex=0)))
   n4 = Get CurrentPosY()
   > n5 = Add Static Mesh Component(bManualAttachment=false, RelativeTransform=MakeTransform!(Location=MakeVector(Y=@n4), Rotation:Rotator="0, 0, 0", Scale) -> (ReturnValue))
   > Set Relative Rotation!(
    execute
    self:Scene Component Object Reference=@n5
    NewRotation_Roll:float=0
    NewRotation_Pitch:float=0
    NewRotation_Yaw:float=Get YawCustomRotation()
    bSweep:bool=false
    bTeleport:bool=false
   ) -> (then, SweepHitResult:Hit Result Structure) {is=title K2Node_CallFunction}
   > n6 = Add Static Mesh Component(bManualAttachment=true)
   > Attach Component To Component(
    self=@n6
    Parent=@n5
    SocketName=None
    LocationRule=KeepRelative
    RotationRule=KeepRelative
    ScaleRule=KeepRelative
    bWeldSimulatedBodies=TRUE
   )
   n7 = Get FilteredMeshIDs()
   n8 = Get ControlID() -> (ControlID=0)
   n9 = Break STRUCT Atom Processor Comparison Utility(STRUCT_AtomProcessorComparisonUtility="Get (a ref)"(Array=Get AtomMeshes(), Dimension 1="Get (a ref)"(Array:Array of Integers=@n7, Dimension 1=@n8) -> (Output:"Integer (by ref)")))
   n10 = Resolve Soft Reference(Input:Static Mesh Soft Object Reference=@n9.AtomMeshMain_36_#2) -> (Output:Static Mesh Object Reference)
   n11 = GetDisplayName(Object=@n10)
   > Set ComponentTags!(
    execute
    ComponentTags:Array of Names=Make Array("[0]":Name=String To Name(InString=@n11)) -> (Array:Array of Names)
    self:Actor Component Object Reference=@n6
   ) -> (then, Output_Get:Array of Names)
   > Is Valid(InputObject=@n10)
   > SetStaticMesh(self=@n6, NewMesh=@n10)
   > Add(TargetArray=Get ChildComponents(), NewItem=@n6)
   > SetForcedLodModel(self=@n6, NewForcedLodModel=Get LODLevel() -> (LODLevel=0))
   > CenterSceneComponent!(execute, self, StaticMeshComponent:Static Mesh Component Object Reference=@n6) -> (then, UpdatedTransform:Transform) {is=title K2Node_CallFunction}
   > AttachAdditionalParts!(
    execute
    self
    AdditionalMeshes:Array of Static Mesh Soft Object References=@n9.AtomMeshAdditional_38_#3
    AtomStaticMeshComponent:Static Mesh Component Object Reference=@n6
    LODModel:int=Get LODLevel() -> (LODLevel=0)
   ) -> (then) {is=title K2Node_CallFunction}
   n12 = GetComponentBounds!(Component=@n6) -> (Origin_X:double, Origin_Y:double, Origin_Z:double, BoxExtent_X:double, BoxExtent_Y:double, BoxExtent_Z:double, SphereRadius)
   > n13 = Add Text Render Component!(
    execute
    self
    bManualAttachment=true
    RelativeTransform_Location_X:double=@n12
    RelativeTransform_Location_Y:double=@n12.Origin_Y
    RelativeTransform_Location_Z:double=0
    RelativeTransform_Rotation="90,0,0"
    RelativeTransform_Scale="1,1,1"
    ComponentTemplateContext
   ) -> (then, ReturnValue)
   > Attach Component To Component(self=@n13, Parent=@n6, SocketName=None, LocationRule=KeepWorld, RotationRule=KeepWorld, ScaleRule=KeepWorld, bWeldSimulatedBodies=false)
   > SetHorizontalAlignment(self=@n13, Value=EHTA_Center)
   > SetTextRenderColor(self=@n13, Value=Get TextColor())
   n14 = float + float(A=@n12.BoxExtent_X, B=10)
   > Add Local Offset(self=@n13, DeltaLocation=MakeVector(Y=0, Z=float * float(A=@n14, B=-1)), bSweep=false, bTeleport=false)
   > Add(TargetArray=Get ChildComponents(), NewItem=@n13)
   n15 = Get LODLevel() -> (LODLevel=0)
   > SetComponentDisplayInfo(
    TextRenderComponent=@n13
    LOD Index=@n15
    AssetIndex="Get (a ref)"(Array=Get FilteredMeshIDs(), Dimension 1=Get ControlID() -> (ControlID=0))
    OverrideColor=false
    IsAutoLOD="Equal (Integer)"(A=@n15)
    IsLodView=false
   ) {is=title K2Node_CallFunction}
   > Set CurrentPosY(CurrentPosY=float - float(A=Get CurrentPosY(), B=Get BaseSize()))
   > Increment Int(Value=Get ControlID() -> (ControlID=0))
   SetGeneralDisplayInfo!(execute=[@n3.else, @n2.else], self, AtomSelectionValid:bool=@n1, DisplayPositionOffset:double=float * float(A=Get BaseSize(), B=0.5)) -> (then) {is=title K2Node_CallFunction}
   > Return Node()
  ]}
 SetupComponent_LodView {nodes [
   n1 = Get AtomMeshes()
   n2 = Resolve Soft Reference(Input:Atom Model Soft Object Reference=Break STRUCT Atom Processor Comparison Utility(STRUCT_AtomProcessorComparisonUtility="Get (a copy)"(Array:Array of STRUCT Atom Processor Comparison Utility Structures=@n1) -> (Output:STRUCT Atom Processor Comparison Utility Structure))) -> (Output:Atom Model Object Reference)
   GetModelProcessor(ProcessorInstance="Get (a copy)"(Array:Array of Atom Model Processor Instance Structures=Get ModelProcessorInstances(self=@n2)) -> (Output:Atom Model Processor Instance Structure))
   > SetModelProcessor!(
    execute
    ProcessorInstance:"Atom Model Processor Instance Structure (by ref)"
    ModelProcessor:Atom Model Processor Object Reference
    bUseCustomSettings:bool=false
   ) -> (then)
   SetupComponent_LodView() {is=title K2Node_FunctionEntry}
   > n3 = InitializeValues_LodView!(execute, self) -> (then, AtomSelectionValid?:bool) {is=title K2Node_CallFunction}
   > n4 = Branch(Condition=@n3)
   > n5 = For Each Loop(Array:Array of STRUCT Atom Processor Comparison Utility Structures=Get AtomMeshes()) -> (Array Element:STRUCT Atom Processor Comparison Utility Structure)
   > n6 = Set AtomModelToDIsplay(AtomModelToDIsplay:int=@n5.Array Index) -> (Output_Get:int=0)
   n7 = Get ColumnSpacing() -> (ColumnSpacing:double)
   n8 = Get LargestBound()
   > Set CurrentColumnOffset(CurrentColumnOffset:double=float * float(A=int + int(A=@n6, B=~) -> (ReturnValue:double), B=float * float(A=@n7, B=@n8))) -> (Output_Get:double)
   > n9 = For Loop(LastIndex=Get NumLODs() -> (NumLODs=0))
   Set CurrentOffsetSingle(execute=@n9.Completed, CurrentOffsetSingle=0)
   n10 = Add Static Mesh Component!(
    execute=@n9.LoopBody
    self
    bManualAttachment=false
    RelativeTransform_Location_X:double=float * float(A=Get CurrentOffsetSingle() -> (CurrentOffsetSingle:double), B=-1)
    RelativeTransform_Location_Y:double=float * float(A=Get CurrentColumnOffset() -> (CurrentColumnOffset:double), B=-1)
    RelativeTransform_Location_Z:double=Get HeightOffset()
    RelativeTransform_Rotation_Roll:float=0
    RelativeTransform_Rotation_Pitch:float=0
    RelativeTransform_Rotation_Yaw:float=Get YawCustomRotation()
    RelativeTransform_Scale:Vector="1,1,1"
    ComponentTemplateContext
   ) -> (then, ReturnValue)
   n11 = Break STRUCT Atom Processor Comparison Utility(STRUCT_AtomProcessorComparisonUtility="Get (a ref)"(
     Array:Array of STRUCT Atom Processor Comparison Utility Structures=Get AtomMeshes()
     Dimension 1=Get AtomModelToDisplay() -> (AtomModelToDisplay=0)
    ) -> (Output:"STRUCT Atom Processor Comparison Utility Structure (by ref)"))
   n12 = Resolve Soft Reference(Input:Static Mesh Soft Object Reference=@n11.AtomMeshMain_36_#2) -> (Output:Static Mesh Object Reference)
   > SetStaticMesh(self=@n10, NewMesh=@n12)
   > CenterSceneComponent(StaticMeshComponent=@n10)
   > SetForcedLodModel(self=@n10, NewForcedLodModel=@n9)
   n13 = GetDisplayName(Object=@n12)
   > Set ComponentTags(ComponentTags=Make Array("[0]"=String To Name(InString=@n13)), self=@n10)
   > Add(TargetArray=Get ChildComponents(), NewItem=@n10)
   > AttachAdditionalParts(AdditionalMeshes=@n11.AtomMeshAdditional_38_#3, AtomStaticMeshComponent=@n10, LODModel=@n9)
   > n14 = Branch(Condition="Equal (Integer)"(A=@n9))
   n15 = Resolve Soft Reference(Input:Atom Model Soft Object Reference=@n11) -> (Output:Atom Model Object Reference)
   > n16 = CountFastBricks!(execute, self, Model:Atom Model Object Reference=@n15, LOD:int=@n9) -> (then, CountString:String) {is=title K2Node_CallFunction}
   > SetForcedLodModel(self=@n10, NewForcedLodModel=Get LODLevel() -> (LODLevel=0))
   n17 = GetComponentBounds(Component=@n10)
   n18 = MakeVector(Y=0, Z=float * float(A=float + float(A=@n17.BoxExtent_X, B=10), B=-1))
   n19 = MakeVector(X=@n17, Y=@n17.Origin_Y)
   > n20 = AddComponentInfoDisplay!(
    execute
    self
    LOD Index:int=Get LODLevel() -> (LODLevel=0)
    AssetIndex:int=Get AtomModelToDisplay() -> (AtomModelToDisplay=0)
    Location:Vector=@n19
    LocalOffset:Vector=@n18
    AdditionalString:String=Append(A="LOD TRANSITION PREVIEW\r\n", B=@n16)
    OverrideTextColor:bool=true
    OverrideColor:Linear Color Structure=(R=1,G=1,B=1,A=1)
    IsAutoLOD:bool="Equal (Integer)"(A=Get LODLevel() -> (LODLevel=0))
    IsLodView:bool=false
   ) -> (then, TextRenderComponent:Text Render Component Object Reference) {is=title K2Node_CallFunction}
   n21 = float * float(A=2, B=@n17.BoxExtent_Y)
   n22 = GetComponentBounds!(Component=@n20) -> (Origin:Vector, BoxExtent_X, BoxExtent_Y, BoxExtent_Z, SphereRadius)
   > Set CurrentOffsetSingle(CurrentOffsetSingle=float - float!(
     A=Get CurrentOffsetSingle()
     B="Max (Float)"(A=@n21, B=@n22.BoxExtent_Y)
     C:double=Get AdditionalSpacingLODs() -> (AdditionalSpacingLODs:double)
     D:double=200
    ) -> (ReturnValue))
   > Set SpacingFromGeneralDisplay(SpacingFromGeneralDisplay:double="Max (Float)"(A=@n22.BoxExtent_Y, B=@n17.BoxExtent_Y)) -> (Output_Get:double)
   n23 = CountFastBricks(execute=@n14.else, Model=@n15, LOD=@n9)
   > n24 = AddComponentInfoDisplay(
    LOD Index=int - int(A=@n9)
    AssetIndex=Get AtomModelToDisplay() -> (AtomModelToDisplay=0)
    Location=@n19
    LocalOffset=@n18
    AdditionalString=@n23
    OverrideTextColor=false
    IsAutoLOD=false
    IsLodView=true
   )
   n25 = float * float(A=GetComponentBounds(Component=@n24).SphereRadius, B=2)
   n26 = float * float(A=2, B=@n17.BoxExtent_Y)
   > Set CurrentOffsetSingle(CurrentOffsetSingle=float - float(A=Get CurrentOffsetSingle(), B="Max (Float)"(A=@n26, B=@n25), C=10, D=Get AdditionalSpacingLODs()))
   SetGeneralDisplayInfo(
    execute=@n4.else
    AtomSelectionValid=@n3
    DisplayPositionOffset=float * float(A=Get SpacingFromGeneralDisplay() -> (SpacingFromGeneralDisplay:double), B=2)
   )
   > Return Node()
  ]}
 StringCombine {nodes [
   n1 = StringCombine!() -> (then, Base String:String, IDs:Array of Integers, Names:Array of Strings, TriangleCounts:Array of Integers) {is=title K2Node_FunctionEntry}
   > Set NewString(NewString:String=@n1) -> (Output_Get:String)
   > n2 = For Each Loop(Array:Array of Strings=@n1.Names) -> (Array Element:String)
   Return Node!(execute=@n2.Completed, CombinedString:String=Get NewString() -> (NewString:String))
   n3 = "Get (a ref)"(Array:Array of Integers=@n1.TriangleCounts, Dimension 1=@n2.Array Index) -> (Output:"Integer (by ref)")
   n4 = "Get (a ref)"(Array=@n1.IDs, Dimension 1=@n2.Array Index)
   > Set NewString(NewString=Append!(
     A=Get NewString()
     B="To String (Integer)"(InInt=@n4)
     C:String=" | "
     D:String=@n2
     E:String=" - "
     F:String="To String (Integer)"(InInt=@n3)
     G:String=" tris"
     H:String="\r\n"
    ) -> (ReturnValue))
  ]}
 UserConstructionScript {nodes [
   Construction Script()
   > n1 = Branch(Condition=Get RunConstructor() -> (RunConstructor:bool))
   n2 = Branch(execute=[@n1.then, @n7.then], Condition=Get GridView())
   > SetupComponent_LineView() {is=title K2Node_CallFunction}
   SetupComponent_LodView(execute=@n2.else) {is=title K2Node_CallFunction}
   n3 = Get FilteredMeshIDs()
   n4 = Get AtomModelToDisplay() -> (AtomModelToDisplay=0)
   n5 = Branch(Condition=AND Boolean(A=Is Not Empty(TargetArray:Array of Integers=@n3), B=Contains Item(TargetArray=@n3, ItemToFind=@n4)))
   > n6 = Set CurrentIndex(CurrentIndex=Find Item(TargetArray=@n3, ItemToFind=@n4)) -> (Output_Get=0)
   n7 = Set AtomModelToDisplay(execute=[@n8.then, @n6.then], AtomModelToDisplay="Get (a ref)"(Array=@n3, Dimension 1=Get CurrentIndex() -> (CurrentIndex=0))) -> (Output_Get=0)
   n8 = Set CurrentIndex(execute=@n5.else, CurrentIndex=0) -> (Output_Get=0)
  ]}
]
variables [
 AdditionalSpacingLODs {type=number}
 AdditionalSpacingOnGrid {type=number}
 AtomMeshes {type="array<STRUCT_AtomProcessorComparisonUtility>", default []}
 AtomMeshesFB {type="array<STRUCT_AtomProcessorComparisonUtility>", default []}
 AtomModelToDIsplay {type=integer}
 AverageTriangleCount {type=integer}
 BaseSize {type=number}
 Budget {type=integer}
 BudgetedLOD {type=integer}
 ChildComponents {type="array<ActorComponent>", default []}
 ColumnSpacing {type=number, default=2}
 ControlID {type=integer}
 CurrentIndex {type=integer}
 CurrentOffsetSingle {type=number}
 CurrentPosX {type=number}
 CurrentPosY {type=number}
 CurrentPosZ {type=number}
 FilteredMeshIDs {type="array<integer>"}
 FontSize {type=number, default=24}
 GeneralDisplayFontSize {type=number}
 GridView {type=boolean}
 HeightOffset {type=number}
 HighestTriangleCounts {type="array<integer>"}
 HighestTriangleNames {type="array<string>", default []}
 InitialPosX {type=number}
 InitialPosY {type=number}
 InitialPosZ {type=number}
 LargestBound {type=number}
 LastModelIndex {type=integer}
 LoadedPath {type=string, default=""}
 LODLevel {type=integer}
 MeshSizeY {type=number}
 NumLODs {type=integer}
 NumModels {type=integer}
 NumOfColumns {type=integer}
 OverBudgetedColor {type=Color, default=<rgba 0 0 0 0>}
 OverBudgetedIDs {type="array<integer>"}
 RadiusInfo {type=IntVector, default=<0 0 0>}
 RotationIncrement {type=number, default=45}
 RunConstructor {type=boolean}
 ShowOnlyOverBudget {type=boolean}
 TextColor {type=Color, default=<rgba 0 0 0 0>}
 YawCustomRotation {type=number}
]
