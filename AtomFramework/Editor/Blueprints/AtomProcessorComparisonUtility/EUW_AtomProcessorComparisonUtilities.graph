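# Editor utility widget that drives EUB_AtomProcessorComparisonUtility visualizer actors to compare Atom Models built with two processors.
# A validated folder or pasted model path becomes a mesh array, with bounds, radius and triangle-budget stats written onto the selected actor.
# Property views bound to the widget select actor and processors, and preview assets.
# Controls trigger reprocessing, rotation, grid or LOD views, Nanite toggles and LOD stepping, mostly delegated to the actor.
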EditorUtilityWidgetBlueprint /AtomFramework/Editor/Blueprints/AtomProcessorComparisonUtility/EUW_AtomProcessorComparisonUtilities
graphs [
 BTN Refresh Condition {nodes [
   BTN Refresh Condition() {is=title K2Node_FunctionEntry}
   n1 = Get CurrentBP() -> (CurrentBP:EUB Atom Processor Comparison Utility Object Reference)
   n2 = IsEmpty(InString=Get LoadedPath!(self:EUB Atom Processor Comparison Utility Object Reference=@n1) -> (LoadedPath:String))
   > Return Node!(execute, ReturnValue:bool=AND Boolean(A=IsValid(Object=@n1), B=NOT Boolean(A=@n2)))
  ]}
 BTN_HigherLODCondition {nodes [
   BTN_HigherLODCondition() {is=title K2Node_FunctionEntry}
   n1 = Get CurrentBP()
   > Is Valid(InputObject=@n1)
   > Return Node(ReturnValue="integer < integer"(
     A=Get LODLevel!(self:EUB Atom Processor Comparison Utility Object Reference=@n1) -> (LODLevel:int=0)
     B=Get NumLODs!(self:EUB Atom Processor Comparison Utility Object Reference=@n1) -> (NumLODs:int=0)
    ))
  ]}
 BTN_LowerLODCondition {nodes [
   BTN_LowerLODCondition() {is=title K2Node_FunctionEntry}
   n1 = Get CurrentBP()
   > Is Valid(InputObject=@n1)
   > Return Node(ReturnValue="integer > integer"(A=Get LODLevel(self=@n1) -> (LODLevel=0)))
  ]}
 CleanUpDataAssetPath {nodes [
   n1 = CleanUpDataAssetPath!() -> (then, UserPath:String) {is=title K2Node_FunctionEntry}
   > n2 = Branch(Condition=StartsWith(SourceString=@n1, InPrefix=/Script/AtomRuntime.AtomModel))
   n3 = Split(SourceString=Split!(InPath:String=@n1) -> (PathPart:String, FilenamePart:String, ExtensionPart:String), InStr=', SearchDir=FromStart)
   > n4 = DoesDirectoryExist!(execute, DirectoryPath:String=@n3.RightS) -> (then, ReturnValue:bool)
   > Set Valid?(Valid?:bool=@n4) -> (Output_Get:bool)
   > n5 = Set TempPath(TempPath:String=@n3.RightS) -> (Output_Get:String)
   n6 = Branch(execute=[@n5.then, @n10.then], Condition=Get Valid?() -> (Valid?:bool))
   Return Node!(execute=[@n8.then, @n6.then], Path:String=Get TempPath() -> (TempPath:String), Success?:bool=Get Valid?())
   n7 = Append(A="Path is invalid and/or the folder doesn't have Atom Models within it (make sure the path is relative.)", B=Get TempPath())
   n8 = Show Message Dialog(
    execute=@n6.else
    Title=MakeLiteralText(Value="LOCTEXT(\"Atom Paths\")")
    Message=MakeLiteralText(Value="To Text (String)"(InString=@n7))
    MessageType=Ok
    DefaultValue=Ok
    MessageCategory=Warning
   )
   n9 = DoesDirectoryExist(execute=@n2.else, DirectoryPath=@n1)
   > Set Valid?(Valid?=@n9)
   > n10 = Set TempPath(TempPath=@n1)
  ]}
 CorrectionForOldAssetPaths {nodes [
   n1 = CorrectionForOldAssetPaths!() -> (then, Path:String) {is=title K2Node_FunctionEntry}
   n2 = MakeLiteralString(Value=/Mesh)
   > Branch(Condition=Contains(SearchIn=@n1, Substring=@n2, bSearchFromEnd=true))
   Split!(SourceString:String=@n1, InStr:String=@n2, SearchCase:ESearchCase Enum, SearchDir:ESearchDir Enum=FromStart) -> (LeftS:String, RightS:String, ReturnValue:bool)
   Return Node!(execute, PathOut:String)
   Reroute Node(InputPin:String=@n1) -> (OutputPin:String)
  ]}
 CreateNewActor {nodes [
   CreateNewActor() {is=title K2Node_FunctionEntry}
   > n1 = SpawnActor EUB Atom Processor Comparison Utility!(
    execute:Exec
    WorldContextObject:Object Reference=Self-Reference()
    Class:Actor Class Reference=&/EUB_AtomProcessorComparisonUtility.EUB_AtomProcessorComparisonUtility_C
    SpawnTransform:Transform=MakeTransform(Rotation="0, 0, 0")
    CollisionHandlingOverride:ESpawnActorCollisionHandlingMethod Enum=Undefined
    TransformScaleMethod:ESpawnActorScaleMethod Enum=MultiplyWithRoot
    Owner:Actor Object Reference
    Instigator:Pawn Object Reference
   ) -> (then:Exec, ReturnValue:EUB Atom Processor Comparison Utility Object Reference) {is=title K2Node_SpawnActorFromClass}
   > Set CurrentBP(CurrentBP:EUB Atom Processor Comparison Utility Object Reference=@n1) -> (Output_Get:EUB Atom Processor Comparison Utility Object Reference)
  ]}
 EventGraph {nodes [
   UpdateAtomVisualizer()
   n1 = Get CurrentBP()
   > "SetText (Editable Text)"!(
    execute
    self:Editable Text Object Reference=Get INP_AtomModelID() -> (INP_AtomModelID:Editor Utility Editable Text Object Reference)
    InText:Text="To Text (Integer)"(Value=Get AtomModelToDIsplay!(self:EUB Atom Processor Comparison Utility Object Reference=@n1) -> (AtomModelToDIsplay:int=0))
   ) -> (then)
   > UpdateAssetPreview()
   n2 = "OnTextCommitted (Editable Text) (INP_TextSize)"!() -> (OutputDelegate, then, Text:"Text (by ref)", CommitMethod:ETextCommit Enum=Default) {is=title K2Node_ComponentBoundEvent}
   n3 = "To String (Text)"(InText=@n2.Text)
   > Branch(Condition=IsNumeric!(SourceString:String=@n3) -> (ReturnValue:bool))
   n4 = Get CurrentBP()
   > Is Valid(InputObject=@n4)
   > Set FontSize!(execute, FontSize:double=String To Float(InString=@n3), self:EUB Atom Processor Comparison Utility Object Reference=@n4) -> (then, Output_Get:double)
   > UpdateAtomVisualizer()
   n5 = "OnTextCommitted (Editable Text) (INP_HeightOffset)"!() -> (OutputDelegate, then, Text:"Text (by ref)", CommitMethod:ETextCommit Enum=Default) {is=title K2Node_ComponentBoundEvent}
   > Is Valid()
   > Set HeightOffset!(
    execute
    HeightOffset:double=String To Float(InString="To String (Text)"(InText=@n5.Text))
    self:EUB Atom Processor Comparison Utility Object Reference=Get CurrentBP()
   ) -> (then, Output_Get:double)
   > UpdateAtomVisualizer()
   "On Property Changed (SPV_Proc1)"!() -> (OutputDelegate, then, PropertyName:Name=None) {is=title K2Node_ComponentBoundEvent}
   > UpdateAssetPreview()
   > UpdateOtherSettings()
   > SetViewButtonsStates()
   "On Clicked (BTN_RotationClock)"() {is=title K2Node_ComponentBoundEvent}
   n6 = Get CurrentBP()
   > Is Valid(InputObject=@n6)
   > RotateClock(self:EUB Atom Processor Comparison Utility Object Reference=@n6)
   "On Clicked (BTN_RotationAntiClock)"() {is=title K2Node_ComponentBoundEvent}
   n7 = Get CurrentBP()
   > Is Valid(InputObject=@n7)
   > RotateAntiClock(self:EUB Atom Processor Comparison Utility Object Reference=@n7)
   "On Clicked (BTN_RotationReset)"() {is=title K2Node_ComponentBoundEvent}
   n8 = Get CurrentBP()
   > Is Valid(InputObject=@n8)
   > RotationReset(self:EUB Atom Processor Comparison Utility Object Reference=@n8)
   "On Property Changed (SPV_Proc2)"!() -> (OutputDelegate, then, PropertyName:Name=None) {is=title K2Node_ComponentBoundEvent}
   > UpdateAssetPreview()
   > UpdateOtherSettings()
   > SetViewButtonsStates()
   "On Property Changed (SinglePropertyView_VisualizerActor)"!() -> (OutputDelegate, then, PropertyName:Name=None) {is=title K2Node_ComponentBoundEvent}
   > UpdateAssetPreview()
   > UpdateOtherSettings()
   > SetViewButtonsStates()
   "On Clicked (BTN_ProcessAtomModels)"() {is=title K2Node_ComponentBoundEvent}
   n9 = Get CurrentBP()
   > Is Valid(InputObject=@n9)
   > ProcessAtomModels!(
    execute
    ShowDialog:bool=true
    AtomMeshes:Array of STRUCT Atom Processor Comparison Utility Structures=Get AtomMeshes!(self:EUB Atom Processor Comparison Utility Object Reference=@n9) -> (AtomMeshes:Array of STRUCT Atom Processor Comparison Utility Structures)
    __WorldContext:Object Reference
   ) -> (then, NumAssets:int=0)
   > UpdateAtomVisualizer()
   n10 = "On Clicked (BTN_Refresh)"() {is=title K2Node_ComponentBoundEvent}
   n11 = Get CurrentBP()
   Is Valid(exec=[@n10.then, @n46.then], InputObject=@n11)
   > n12 = CleanUpDataAssetPath!(execute, self, UserPath:String=Get LoadedPath(self=@n11)) -> (then, Path:String, Success?:bool) {is=title K2Node_CallFunction}
   > Branch(Condition=@n12.Success?)
   > n13 = Set PathToLoad(PathToLoad:String=@n12) -> (Output_Get:String)
   n14 = Get CurrentBP()
   n15 = Is Valid(exec=[@n29.then, @n13.then], InputObject=@n14)
   n16 = CreateNewActor(execute=@n15.Is Not Valid) {is=title K2Node_CallFunction}
   n17 = Get PathToLoad() -> (PathToLoad:String)
   n18 = CreateAtomMeshesArray!(
    execute=[@n16.then, @n15.Is Valid]
    DirectoryPath:String=@n17
    Processor_1:Atom Model Processor Class Reference=Get Processor_1() -> (Processor_1:Atom Model Processor Class Reference)
    Processor_2:Atom Model Processor Class Reference=Get Processor_2() -> (Processor_2:Atom Model Processor Class Reference)
    Geo1:bool=Get Geo1() -> (Geo1:bool)
    Geo2:bool=Get Geo2() -> (Geo2:bool)
    Process Primitives:bool=Get ProcessPrimitives() -> (ProcessPrimitives:bool)
    __WorldContext:Object Reference
   ) -> (then, AtomMeshes:Array of STRUCT Atom Processor Comparison Utility Structures)
   > Set AtomMeshes!(
    execute
    AtomMeshes:Array of STRUCT Atom Processor Comparison Utility Structures=@n18
    self:EUB Atom Processor Comparison Utility Object Reference=@n14
   ) -> (then, Output_Get:Array of STRUCT Atom Processor Comparison Utility Structures)
   > Set LoadedPath!(execute, LoadedPath:String=@n17, self:EUB Atom Processor Comparison Utility Object Reference=@n14) -> (then, Output_Get:String)
   n19 = Get AtomMeshes(self=@n14)
   > n20 = GetLargestBound!(execute, Array:Array of STRUCT Atom Processor Comparison Utility Structures=@n19, __WorldContext:Object Reference) -> (then, LargestBound:double, LargestBoundMesh:Static Mesh Object Reference)
   > Set LargestBound!(execute, LargestBound:double=@n20, self:EUB Atom Processor Comparison Utility Object Reference=@n14) -> (then, Output_Get:double)
   > n21 = GetRadiusInfo!(execute, AtomMeshes:Array of STRUCT Atom Processor Comparison Utility Structures=@n19, __WorldContext:Object Reference) -> (then, AverageRadius:int=0, LargestRadius:int=0, SmallestRadius:int=0)
   > Set RadiusInfo!(
    execute
    RadiusInfo:Int Vector Structure=Make Int Vector!(X:int=@n21, Y:int=@n21.LargestRadius, Z:int=@n21.SmallestRadius) -> (IntVector:Int Vector Structure) {is=title K2Node_MakeStruct}
    self:EUB Atom Processor Comparison Utility Object Reference=@n14
   ) -> (then, Output_Get:Int Vector Structure)
   > UpdateAtomVisualizer()
   > UpdateAssetPreview()
   > UpdateOtherSettings()
   > Set Text(self:Editable Text Object Reference=Get INP_Path() -> (INP_Path:Editor Utility Editable Text Object Reference))
   > Set Text(
    Text="To Text (String)"(InString=Get LoadedPath(self=@n14))
    self:Text Object Reference=Get TXT_CurrentlyLoaded() -> (TXT_CurrentlyLoaded:Text Object Reference)
   )
   n22 = "OnTextCommitted (Editable Text) (INP_RotationInc)"!() -> (OutputDelegate, then, Text:"Text (by ref)", CommitMethod:ETextCommit Enum=Default) {is=title K2Node_ComponentBoundEvent}
   n23 = Get CurrentBP()
   > Is Valid(InputObject=@n23)
   > Set RotationIncrement!(
    execute
    RotationIncrement:double=String To Float(InString="To String (Text)"(InText=@n22.Text))
    self:EUB Atom Processor Comparison Utility Object Reference=@n23
   ) -> (then, Output_Get:double)
   "On Clicked (BTN_ProcessAtomModel)"() {is=title K2Node_ComponentBoundEvent}
   > ProcessAtomModel!(
    execute
    Atom Model:Atom Model Soft Object Reference=Make Soft Reference!(Input:Atom Model Object Reference=Get AtomModel() -> (AtomModel:Atom Model Object Reference)) -> (Output:Atom Model Soft Object Reference)
    ShowDialog:bool=true
    __WorldContext:Object Reference
   ) -> (then, Success?:bool)
   > UpdateAtomVisualizer()
   "On Clicked (BTN_CreateNewActor)"() {is=title K2Node_ComponentBoundEvent}
   > CreateNewActor()
   > UpdateOtherSettings()
   > SetViewButtonsStates()
   "On Clicked (BTN_DeleteActor)"() {is=title K2Node_ComponentBoundEvent}
   n24 = Get CurrentBP()
   > Is Valid(InputObject=@n24)
   n25 = Append!(A="Are you sure you want to delete ", B=GetDisplayName(Object=@n24), C:String=" ?") -> (ReturnValue)
   > n26 = Show Message Dialog(
    Title=MakeLiteralText(Value="LOCTEXT(\"Delete Selected Actor?\")")
    Message=MakeLiteralText(Value="To Text (String)"(InString=@n25))
    MessageType=YesNo
    DefaultValue=No
    MessageCategory=Warning
   )
   > Branch(Condition="Equal (Enum)"(A:EAppReturnType Enum=@n26, B:EAppReturnType Enum=Yes))
   > Destroy Actor(self=Get CurrentBP())
   > SyncActors()
   "On Clicked (BTN_SearchAtomModels)"() {is=title K2Node_ComponentBoundEvent}
   n27 = Get INP_Path()
   > n28 = CleanUpDataAssetPath(UserPath="To String (Text)"(InText="GetText (Editable Text)"!(self:Editable Text Object Reference=@n27) -> (ReturnValue:Text)))
   > Branch(Condition=@n28.Success?)
   > n29 = Set PathToLoad(PathToLoad=@n28)
   n30 = "On Check State Changed (CHB_Custom_Geo_1)"!() -> (OutputDelegate, then, bIsChecked:bool) {is=title K2Node_ComponentBoundEvent}
   > Set Geo1(Geo1:bool=@n30.bIsChecked) -> (Output_Get:bool)
   n31 = "On Check State Changed (CHB_Custom_Geo_2)"!() -> (OutputDelegate, then, bIsChecked:bool) {is=title K2Node_ComponentBoundEvent}
   > Set Geo2(Geo2:bool=@n31.bIsChecked) -> (Output_Get:bool)
   n32 = "On Check State Changed (CHB_ShowOver)"!() -> (OutputDelegate, then, bIsChecked:bool) {is=title K2Node_ComponentBoundEvent}
   > Set ShowOnlyOverBudget!(execute, ShowOnlyOverBudget:bool=@n32.bIsChecked, self:EUB Atom Processor Comparison Utility Object Reference=@n1) -> (then, Output_Get:bool)
   > Set CurrentIndex!(execute, CurrentIndex:int=0, self:EUB Atom Processor Comparison Utility Object Reference=@n1) -> (then, Output_Get:int=0)
   n33 = "On Check State Changed (CHB_ReimportPrimitives)"!() -> (OutputDelegate, then, bIsChecked:bool) {is=title K2Node_ComponentBoundEvent}
   > Set ProcessPrimitives(ProcessPrimitives:bool=@n33.bIsChecked) -> (Output_Get:bool)
   "On Property Changed (DetailsView_AssetsPreview)"!() -> (OutputDelegate, then, PropertyName:Name=None) {is=title K2Node_ComponentBoundEvent}
   > UpdateAssetPreview()
   "On Clicked (BTN_GridView)"() {is=title K2Node_ComponentBoundEvent}
   n34 = Get CurrentBP()
   > Is Valid(InputObject=@n34)
   > n35 = "- GRID ViewMode - "(self:EUB Atom Processor Comparison Utility Object Reference=@n34)
   SetViewButtonsStates(execute=[@n39.then, @n35.then])
   n36 = "OnTextCommitted (Editable Text) (INP_RowSpacing)"!() -> (OutputDelegate, then, Text:"Text (by ref)", CommitMethod:ETextCommit Enum=Default) {is=title K2Node_ComponentBoundEvent}
   n37 = "To String (Text)"(InText=@n36.Text)
   > Branch(Condition=IsNumeric(SourceString=@n37))
   > Set ColumnSpacing!(execute, ColumnSpacing:double=String To Float(InString=@n37), self:EUB Atom Processor Comparison Utility Object Reference=Get CurrentBP()) -> (then, Output_Get:double)
   > UpdateAtomVisualizer()
   "On Clicked (BTN_LODView)"() {is=title K2Node_ComponentBoundEvent}
   n38 = Get CurrentBP()
   > Is Valid(InputObject=@n38)
   > n39 = - LOD ViewMode -(self:EUB Atom Processor Comparison Utility Object Reference=@n38)
   "On Clicked (BTN_DisableNanite)"() {is=title K2Node_ComponentBoundEvent}
   n40 = Get CurrentBP()
   > Is Valid(InputObject=@n40)
   > Disable Nanite/LOD Auto(self:EUB Atom Processor Comparison Utility Object Reference=@n40)
   > InitializeVisualizer()
   "On Clicked (BTN_EnableNanite)"() {is=title K2Node_ComponentBoundEvent}
   n41 = Get CurrentBP()
   > Is Valid(InputObject=@n41)
   > Enable Nanite(self:EUB Atom Processor Comparison Utility Object Reference=@n41)
   > InitializeVisualizer()
   "On Clicked (BTN_LODLower)"() {is=title K2Node_ComponentBoundEvent}
   n42 = Get CurrentBP()
   > Is Valid(InputObject=@n42)
   > LOD Bias Previous(self:EUB Atom Processor Comparison Utility Object Reference=@n42)
   > InitializeVisualizer()
   "On Clicked (BTN_LODHigher)"() {is=title K2Node_ComponentBoundEvent}
   n43 = Get CurrentBP()
   > Is Valid(InputObject=@n43)
   > LOD Bias Next(self:EUB Atom Processor Comparison Utility Object Reference=@n43)
   > InitializeVisualizer()
   SyncActors() {is=title K2Node_CustomEvent}
   > n44 = GetAllActorsOfClass!(
    execute
    WorldContextObject:Object Reference
    ActorClass:Actor Class Reference=&EUB_AtomProcessorComparisonUtility_C
   ) -> (then, OutActors:Array of EUB Atom Processor Comparison Utility Object References)
   > n45 = Branch(Condition=Is Not Empty!(TargetArray:Array of EUB Atom Processor Comparison Utility Object References=@n44) -> (ReturnValue:bool))
   > Set CurrentBP(CurrentBP="Get (a copy)"(Array:Array of EUB Atom Processor Comparison Utility Object References=@n44) -> (Output:EUB Atom Processor Comparison Utility Object Reference))
   Set CurrentBP(execute=@n45.else)
   n46 = Refresh() {is=title K2Node_CustomEvent}
   UpdateAssetPreview() {is=title K2Node_CustomEvent}
   > Clear(TargetArray:Array of Static Mesh Soft Object References=Get AtomMeshes!() -> (AtomMeshes:Array of Static Mesh Soft Object References))
   n47 = Get CurrentBP()
   n48 = Break STRUCT Atom Processor Comparison Utility!(STRUCT_AtomProcessorComparisonUtility:"STRUCT Atom Processor Comparison Utility Structure (by ref)"="Get (a ref)"(
     Array:Array of STRUCT Atom Processor Comparison Utility Structures=Get AtomMeshes!(self:EUB Atom Processor Comparison Utility Object Reference=@n47) -> (AtomMeshes:Array of STRUCT Atom Processor Comparison Utility Structures)
     Dimension 1=Get AtomModelToDIsplay(self=@n47) -> (AtomModelToDIsplay=0)
    ) -> (Output:"STRUCT Atom Processor Comparison Utility Structure (by ref)")) -> (
    AtomModel_35_#1:Atom Model Soft Object Reference
    AtomMeshMain_36_#2:Static Mesh Soft Object Reference
    AtomMeshAdditional_38_#3:Array of Static Mesh Soft Object References
   ) {is=title K2Node_BreakStruct}
   > n49 = LoadAsset_Blocking(Asset=@n48)
   > n50 = Cast To AtomModel!(execute, Object=@n49) -> (then, CastFailed, AsAtom Model:Atom Model Object Reference) {is=title K2Node_DynamicCast}
   > n51 = Set AtomModel(AtomModel:Atom Model Object Reference=@n50) -> (Output_Get:Atom Model Object Reference)
   n52 = For Each Loop(
    Exec=[@n51.then, @n50.CastFailed]
    Array:Array of Static Mesh Soft Object References=@n48.AtomMeshAdditional_38_#3
   ) -> (Array Element:Static Mesh Soft Object Reference)
   n53 = Get AtomMeshes!() -> (AtomMeshes:Array of Static Mesh Soft Object References)
   Add Unique(
    execute=@n52.Completed
    TargetArray:Array of Static Mesh Soft Object References=@n53
    NewItem:"Static Mesh Soft Object Reference (by ref)"=@n48.AtomMeshMain_36_#2
   )
   > Set AtomModelList(AtomModelList:Array of STRUCT Atom Processor Comparison Utility Structures=Get AtomMeshes!(self:EUB Atom Processor Comparison Utility Object Reference=Get CurrentBP()) -> (AtomMeshes:Array of STRUCT Atom Processor Comparison Utility Structures)) -> (Output_Get:Array of STRUCT Atom Processor Comparison Utility Structures)
   Add Unique(execute=@n52.LoopBody, TargetArray=@n53, NewItem=@n52)
   n54 = Event PreConstruct()
   > n55 = Branch(Condition=@n54.IsDesignTime)
   SetObject(
    execute=[@n56.then, @n55.then]
    self=[
     Get SPV_Proc1() -> (SPV_Proc1:Single Property View Object Reference)
     Get SinglePropertyView_VisualizerActor() -> (SinglePropertyView_VisualizerActor:Single Property View Object Reference)
     Get DetailsView_AssetsPreview() -> (DetailsView_AssetsPreview:Details View Object Reference)
     Get SPV_Proc2() -> (SPV_Proc2:Single Property View Object Reference)
    ]
    NewObject=Self-Reference()
   )
   > SyncActors() {is=title K2Node_CallFunction}
   > Refresh() {is=title K2Node_CallFunction}
   > SetViewButtonsStates()
   n56 = Event Construct()
  ]}
 InitializeVisualizer {nodes [
   InitializeVisualizer() {is=title K2Node_FunctionEntry}
   n1 = Get CurrentBP()
   > n2 = GetFilteredMeshes!(
    execute
    AtomMeshes:Array of STRUCT Atom Processor Comparison Utility Structures=Get AtomMeshes(self=@n1)
    BudgetedLOD:int=Get BudgetedLOD!(self:EUB Atom Processor Comparison Utility Object Reference=@n1) -> (BudgetedLOD:int=0)
    TriangleBudget:int=Get Budget!(self:EUB Atom Processor Comparison Utility Object Reference=@n1) -> (Budget:int=0)
    OnlyOverBudget:bool=Get ShowOnlyOverBudget!(self:EUB Atom Processor Comparison Utility Object Reference=@n1) -> (ShowOnlyOverBudget:bool)
    __WorldContext:Object Reference
   ) -> (
    then
    FilteredMeshIDs:Array of Integers
    OverBudgetedIDs:Array of Integers
    HighestTriangleCounts:Array of Integers
    HighestTriangleNames:Array of Strings
   )
   n3 = Get CurrentBP()
   > Set FilteredMeshIDs!(execute, FilteredMeshIDs:Array of Integers=@n2, self:EUB Atom Processor Comparison Utility Object Reference=@n3) -> (then, Output_Get:Array of Integers)
   > Set OverBudgetedIDs!(execute, OverBudgetedIDs:Array of Integers=@n2.OverBudgetedIDs, self:EUB Atom Processor Comparison Utility Object Reference=@n3) -> (then, Output_Get:Array of Integers)
   > Set HighestTriangleCounts!(execute, HighestTriangleCounts:Array of Integers=@n2.HighestTriangleCounts, self:EUB Atom Processor Comparison Utility Object Reference=@n3) -> (then, Output_Get:Array of Integers)
   > Set HighestTriangleNames!(execute, HighestTriangleNames:Array of Strings=@n2.HighestTriangleNames, self:EUB Atom Processor Comparison Utility Object Reference=@n3) -> (then, Output_Get:Array of Strings)
   > n4 = GetTriangleCountAtomMesh!(
    execute
    AtomMeshes:Array of STRUCT Atom Processor Comparison Utility Structures=Get AtomMeshes(self=@n3)
    LODIndex:int=Get LODLevel(self=@n3) -> (LODLevel=0)
    __WorldContext:Object Reference
   ) -> (then, TotalTriangleCount:int=0, AverageTriangleCount:int=0, NumModels:int=0)
   > Set AverageTriangleCount!(execute, AverageTriangleCount:int=@n4.AverageTriangleCount, self:EUB Atom Processor Comparison Utility Object Reference=@n3) -> (then, Output_Get:int=0)
   > Return Node!(execute)
  ]}
 SetViewButtonsStates {nodes [
   SetViewButtonsStates() {is=title K2Node_FunctionEntry}
   > Set bIsEnabled!(
    execute
    bIsEnabled:bool=NOT Boolean(A=false)
    self:Widget Object Reference=Get BTN_GridView() -> (BTN_GridView:Editor Utility Button Object Reference)
   ) -> (then, Output_Get:bool)
   n1 = Get BTN_LODView() -> (BTN_LODView:Editor Utility Button Object Reference)
   > Set bIsEnabled(bIsEnabled=Get GridView!(self:EUB Atom Processor Comparison Utility Object Reference=Get CurrentBP()) -> (GridView:bool), self=@n1)
   > n2 = Branch(Condition=Get bIsEnabled!(self:Widget Object Reference=@n1) -> (bIsEnabled:bool))
   n3 = Get BTN_LODView()
   > SetBackgroundColor!(execute, self:Button Object Reference=@n3, InBackgroundColor:Linear Color Structure=(R=1,G=1,B=1,A=1)) -> (then)
   n4 = Get BTN_GridView()
   > SetBackgroundColor(self=@n4, InBackgroundColor=(R=0.2188,G=0.2188,B=0.2188,A=1))
   SetBackgroundColor(execute=@n2.else, self=@n3, InBackgroundColor=(R=0.2188,G=0.2188,B=0.2188,A=1))
   > SetBackgroundColor(self=@n4, InBackgroundColor=(R=1,G=1,B=1,A=1))
  ]}
 UpdateAtomVisualizer {nodes [
   UpdateAtomVisualizer() {is=title K2Node_FunctionEntry}
   > InitializeVisualizer() {is=title K2Node_CallFunction}
   > UpdateInfo(self:EUB Atom Processor Comparison Utility Object Reference=Get CurrentBP())
  ]}
 UpdateOtherSettings {nodes [
   UpdateOtherSettings() {is=title K2Node_FunctionEntry}
   n1 = Get CurrentBP()
   n2 = Get RotationIncrement!(self:EUB Atom Processor Comparison Utility Object Reference=@n1) -> (RotationIncrement:double)
   > "SetText (Editable Text)"(
    self=Get INP_RotationInc() -> (INP_RotationInc:Editor Utility Editable Text Object Reference)
    InText="To Text (Integer)"(Value=Truncate(A=@n2))
   )
   n3 = Get HeightOffset!(self:EUB Atom Processor Comparison Utility Object Reference=@n1) -> (HeightOffset:double)
   > "SetText (Editable Text)"(
    self=Get INP_HeightOffset() -> (INP_HeightOffset:Editor Utility Editable Text Object Reference)
    InText="To Text (Integer)"(Value=Truncate(A=@n3))
   )
   > SetIsChecked!(
    execute
    self:Check Box Object Reference=Get CHB_ShowOver() -> (CHB_ShowOver:Editor Utility Check Box Object Reference)
    InIsChecked:bool=Get ShowOnlyOverBudget(self=@n1)
   ) -> (then)
   > Set Text(Text="To Text (String)"(InString=Get LoadedPath(self=@n1)), self:Editable Text Object Reference=Get INP_Path())
  ]}
]
variables [
 AtomMeshes {type="array<StaticMesh>", default []}
 AtomModel {type=AtomModel, default=null}
 AtomModelList {type="array<STRUCT_AtomProcessorComparisonUtility>", default []}
 CurrentBP {type=EUB_AtomProcessorComparisonUtility_C, default=null}
 Geo1 {type=boolean}
 Geo2 {type=boolean}
 PathToLoad {type=string, default=""}
 Processor_1 {type=Class, default=null}
 Processor_2 {type=Class, default=null}
 ProcessPrimitives {type=boolean}
]
widgets {
 info {namedSlotCount=2, rootWidgetClass=&EditorUtilityScrollBox, widgetCount=155}
 widgets [
  EditorUtilityScrollBox_Vertical {widgetClassPath=&EditorUtilityScrollBox, widget=EditorUtilityScrollBox_Vertical, properties {scrollbarThickness=<5 5>}}
  VerticalBox_0 {
   widgetClassPath=&VerticalBox
   parent=EditorUtilityScrollBox_Vertical
   widget=VerticalBox_0
   slot {class=ScrollBoxSlot, properties {padding=<ltrb 10 10 10 10>}}
  }
  Spacer_82 {widgetClassPath=&Spacer, parent=VerticalBox_0, widget=Spacer_82, properties {size=<1 10>}, slot {class=VerticalBoxSlot}}
  TextBlock_94 {
   widgetClassPath=&TextBlock
   parent=VerticalBox_0
   widget=TextBlock_94
   properties {
    justification=Center
    text=">>> ATOM PROCESSOR COMPARISON UTILITY <<<"
    colorAndOpacity {colorUseRule=UseColor_Specified, specifiedColor=<rgba 0.5647 0.5647 0.5647 1>}
    font {
     bForceMonospaced=false
     bMaterialIsStencil=false
     fontMaterial=null
     fontObject=&Roboto
     letterSpacing=-20
     monospacedWidth=1
     size=13.5
     skewAmount=0
     typefaceFontName=Bold
     outlineSettings {
      bApplyOutlineToDropShadows=true
      bMiteredCorners=false
      bSeparateFillAlpha=false
      outlineMaterial=null
      outlineSize=0
      outlineColor=<rgba 0 0 0 1>
     }
    }
    shadowColorAndOpacity=<rgba 0 0 0 0.5>
   }
   slot {class=VerticalBoxSlot, properties {horizontalAlignment=HAlign_Center, verticalAlignment=VAlign_Center}}
  }
  Spacer_33 {widgetClassPath=&Spacer, parent=VerticalBox_0, widget=Spacer_33, properties {size=<1 10>}, slot {class=VerticalBoxSlot}}
  Border_Line_5 {
   widgetClassPath=&Border
   parent=VerticalBox_0
   widget=Border_Line_5
   properties {
    background {
     bIsDynamicallyLoaded=false
     drawAs=Box
     imageType=NoImage
     mirroring=NoMirror
     resourceName=None
     resourceObject=null
     tiling=NoTile
     imageSize=<32 32>
     margin=<ltrb 0 0 0 0>
     outlineSettings {
      bUseBrushTransparency=false
      roundingType=HalfHeightRadius
      width=0
      color {colorUseRule=UseColor_Specified, specifiedColor=<rgba 0 0 0 0>}
      cornerRadii=<0 0 0 0>
     }
     tintColor {colorUseRule=UseColor_Specified, specifiedColor=<rgba 1 1 1 0.25>}
     uVRegion {bIsValid=false, max=<0 0>, min=<0 0>}
    }
    desiredSizeScale=<1 0.3>
    padding=<ltrb 4 4 4 4>
   }
   slot {class=VerticalBoxSlot}
  }
  Spacer_74 {widgetClassPath=&Spacer, parent=VerticalBox_0, widget=Spacer_74, properties {size=<1 10>}, slot {class=VerticalBoxSlot}}
  HorizontalBox-AtomActors {widgetClassPath=&HorizontalBox, parent=VerticalBox_0, widget=HorizontalBox-AtomActors, slot {class=VerticalBoxSlot}}
  Spacer_19 {
   widgetClassPath=&Spacer
   parent=HorizontalBox-AtomActors
   widget=Spacer_19
   properties {size=<10 0>}
   slot {class=HorizontalBoxSlot}
  }
  VerticalBox_3 {widgetClassPath=&VerticalBox, parent=HorizontalBox-AtomActors, widget=VerticalBox_3, slot {class=HorizontalBoxSlot}}
  SinglePropertyView_VisualizerActor {
   widgetClassPath=&SinglePropertyView
   parent=VerticalBox_3
   widget=SinglePropertyView_VisualizerActor
   properties {nameOverride=Atom Visualizer Actors, propertyName=CurrentBP}
   slot {class=VerticalBoxSlot}
  }
  Spacer_203 {widgetClassPath=&Spacer, parent=VerticalBox_3, widget=Spacer_203, properties {size=<1 10>}, slot {class=VerticalBoxSlot}}
  HorizontalBox_0 {
   widgetClassPath=&HorizontalBox
   parent=VerticalBox_3
   widget=HorizontalBox_0
   slot {class=VerticalBoxSlot, properties {horizontalAlignment=HAlign_Left}}
  }
  BTN_CreateNewActor {
   widgetClassPath=&EditorUtilityButton
   parent=HorizontalBox_0
   widget=BTN_CreateNewActor
   properties {toolTipText=Creates a new instance of the Atom Visualizer Actor in the map}
   slot {class=HorizontalBoxSlot}
  }
  TextBlock_14 {
   widgetClassPath=&TextBlock
   parent=BTN_CreateNewActor
   widget=TextBlock_14
   properties {
    text=Create New
    colorAndOpacity {colorUseRule=UseColor_Specified, specifiedColor=<rgba 0.5647 0.5647 0.5647 1>}
    font {
     bForceMonospaced=false
     bMaterialIsStencil=false
     fontMaterial=null
     fontObject=&Roboto
     letterSpacing=0
     monospacedWidth=1
     size=7.5
     skewAmount=0
     typefaceFontName=Regular
     outlineSettings {
      bApplyOutlineToDropShadows=false
      bMiteredCorners=false
      bSeparateFillAlpha=false
      outlineMaterial=null
      outlineSize=0
      outlineColor=<rgba 0 0 0 1>
     }
    }
   }
   slot {class=ButtonSlot}
  }
  Spacer_16 {widgetClassPath=&Spacer, parent=HorizontalBox_0, widget=Spacer_16, properties {size=<5 5>}, slot {class=HorizontalBoxSlot}}
  BTN_DeleteActor {
   widgetClassPath=&EditorUtilityButton
   parent=HorizontalBox_0
   widget=BTN_DeleteActor
   properties {toolTipText=Delete current Atom Visualizer actor}
   slot {class=HorizontalBoxSlot, properties {horizontalAlignment=HAlign_Left}}
  }
  TextBlock_23 {
   widgetClassPath=&TextBlock
   parent=BTN_DeleteActor
   widget=TextBlock_23
   properties {
    text=Delete Selected
    colorAndOpacity {colorUseRule=UseColor_Specified, specifiedColor=<rgba 0.5647 0.5647 0.5647 1>}
    font {
     bForceMonospaced=false
     bMaterialIsStencil=false
     fontMaterial=null
     fontObject=&Roboto
     letterSpacing=0
     monospacedWidth=1
     size=7.5
     skewAmount=0
     typefaceFontName=Regular
     outlineSettings {
      bApplyOutlineToDropShadows=false
      bMiteredCorners=false
      bSeparateFillAlpha=false
      outlineMaterial=null
      outlineSize=0
      outlineColor=<rgba 0 0 0 1>
     }
    }
   }
   slot {class=ButtonSlot}
  }
  Spacer_15 {
   widgetClassPath=&Spacer
   parent=HorizontalBox-AtomActors
   widget=Spacer_15
   properties {size=<10 0>}
   slot {class=HorizontalBoxSlot}
  }
  BTN_Refresh {
   widgetClassPath=&EditorUtilityButton
   parent=HorizontalBox-AtomActors
   widget=BTN_Refresh
   properties {
    toolTipText=Updates the current actor to reflect any changes made to its models on their respective info display.
    widgetStyle {
     clickedSlateSound {resourceObject=null}
     disabled {
      bIsDynamicallyLoaded=false
      drawAs=RoundedBox
      imageType=NoImage
      mirroring=NoMirror
      resourceName=None
      resourceObject=null
      tiling=NoTile
      imageSize=<0 0>
      margin=<ltrb 0 0 0 0>
      outlineSettings {
       bUseBrushTransparency=false
       roundingType=FixedRadius
       width=1
       color {colorUseRule=UseColor_ColorTable, specifiedColor=<rgba 0 0 0 0>}
       cornerRadii=<4 4 4 4>
      }
      tintColor {colorUseRule=UseColor_Specified, specifiedColor=<rgba 0.0395 0.0395 0.0395 1>}
      uVRegion {bIsValid=false, max=<0 0>, min=<0 0>}
     }
     disabledForeground {colorUseRule=UseColor_ColorTable, specifiedColor=<rgba 0 0 0 0>}
     hovered {
      bIsDynamicallyLoaded=false
      drawAs=RoundedBox
      imageType=NoImage
      mirroring=NoMirror
      resourceName=None
      resourceObject=null
      tiling=NoTile
      imageSize=<32 32>
      margin=<ltrb 0 0 0 0>
      outlineSettings {
       bUseBrushTransparency=false
       roundingType=FixedRadius
       width=1
       color {colorUseRule=UseColor_ColorTable, specifiedColor=<rgba 0 0 0 0>}
       cornerRadii=<4 4 4 4>
      }
      tintColor {colorUseRule=UseColor_Specified, specifiedColor=<rgba 0.0953 0.0953 0.0953 1>}
      uVRegion {bIsValid=false, max=<0 0>, min=<0 0>}
     }
     hoveredForeground {colorUseRule=UseColor_ColorTable, specifiedColor=<rgba 0 0 0 0>}
     hoveredSlateSound {resourceObject=null}
     normal {
      bIsDynamicallyLoaded=false
      drawAs=RoundedBox
      imageType=NoImage
      mirroring=NoMirror
      resourceName=None
      resourceObject=null
      tiling=NoTile
      imageSize=<32 32>
      margin=<ltrb 0 0 0 0>
      outlineSettings {
       bUseBrushTransparency=false
       roundingType=FixedRadius
       width=1
       color {colorUseRule=UseColor_ColorTable, specifiedColor=<rgba 0 0 0 0>}
       cornerRadii=<4 4 4 4>
      }
      tintColor {colorUseRule=UseColor_Specified, specifiedColor=<rgba 0.0395 0.0395 0.0395 1>}
      uVRegion {bIsValid=false, max=<0 0>, min=<0 0>}
     }
     normalForeground {colorUseRule=UseColor_ColorTable, specifiedColor=<rgba 0 0 0 0>}
     normalPadding=<ltrb 12 7 12 7>
     pressed {
      bIsDynamicallyLoaded=false
      drawAs=RoundedBox
      imageType=NoImage
      mirroring=NoMirror
      resourceName=None
      resourceObject=null
      tiling=NoTile
      imageSize=<32 32>
      margin=<ltrb 0 0 0 0>
      outlineSettings {
       bUseBrushTransparency=false
       roundingType=FixedRadius
       width=1
       color {colorUseRule=UseColor_ColorTable, specifiedColor=<rgba 0 0 0 0>}
       cornerRadii=<4 4 4 4>
      }
      tintColor {colorUseRule=UseColor_Specified, specifiedColor=<rgba 0.0284 0.0284 0.0284 1>}
      uVRegion {bIsValid=false, max=<0 0>, min=<0 0>}
     }
     pressedForeground {colorUseRule=UseColor_ColorTable, specifiedColor=<rgba 0 0 0 0>}
     pressedPadding=<ltrb 12 2.5 12 0.5>
     pressedSlateSound {resourceObject=null}
    }
   }
   slot {class=HorizontalBoxSlot, properties {verticalAlignment=VAlign_Center}}
  }
  TextBlock_12 {
   widgetClassPath=&TextBlock
   parent=BTN_Refresh
   widget=TextBlock_12
   properties {
    text=REFRESH
    colorAndOpacity {colorUseRule=UseColor_Specified, specifiedColor=<rgba 0.5647 0.5647 0.5647 1>}
    font {
     bForceMonospaced=false
     bMaterialIsStencil=false
     fontMaterial=null
     fontObject=&Roboto
     letterSpacing=0
     monospacedWidth=1
     size=9
     skewAmount=0
     typefaceFontName=Bold
     outlineSettings {
      bApplyOutlineToDropShadows=false
      bMiteredCorners=false
      bSeparateFillAlpha=false
      outlineMaterial=null
      outlineSize=0
      outlineColor=<rgba 0 0 0 1>
     }
    }
   }
   slot {class=ButtonSlot}
  }
  Spacer_20 {
   widgetClassPath=&Spacer
   parent=HorizontalBox-AtomActors
   widget=Spacer_20
   properties {size=<10 0>}
   slot {class=HorizontalBoxSlot, properties {size {sizeRule=Fill, value=1}}}
  }
  Spacer_25 {widgetClassPath=&Spacer, parent=VerticalBox_0, widget=Spacer_25, properties {size=<1 10>}, slot {class=VerticalBoxSlot}}
  Border_Line {
   widgetClassPath=&Border
   parent=VerticalBox_0
   widget=Border_Line
   properties {
    background {
     bIsDynamicallyLoaded=false
     drawAs=Box
     imageType=NoImage
     mirroring=NoMirror
     resourceName=None
     resourceObject=null
     tiling=NoTile
     imageSize=<32 32>
     margin=<ltrb 0 0 0 0>
     outlineSettings {
      bUseBrushTransparency=false
      roundingType=HalfHeightRadius
      width=0
      color {colorUseRule=UseColor_Specified, specifiedColor=<rgba 0 0 0 0>}
      cornerRadii=<0 0 0 0>
     }
     tintColor {colorUseRule=UseColor_Specified, specifiedColor=<rgba 1 1 1 0.25>}
     uVRegion {bIsValid=false, max=<0 0>, min=<0 0>}
    }
    desiredSizeScale=<1 0.3>
    padding=<ltrb 4.1 4.1 4.1 4.1>
   }
   slot {class=VerticalBoxSlot}
  }
  Spacer_26 {widgetClassPath=&Spacer, parent=VerticalBox_0, widget=Spacer_26, properties {size=<1 10>}, slot {class=VerticalBoxSlot}}
  VerticalBoxDataAsset {widgetClassPath=&VerticalBox, parent=VerticalBox_0, widget=VerticalBoxDataAsset, slot {class=VerticalBoxSlot}}
  HorizontalBox-AtomDataAssets {widgetClassPath=&HorizontalBox, parent=VerticalBoxDataAsset, widget=HorizontalBox-AtomDataAssets, slot {class=VerticalBoxSlot}}
  VerticalBox-DataAssetCreation {
   widgetClassPath=&VerticalBox
   parent=HorizontalBox-AtomDataAssets
   widget=VerticalBox-DataAssetCreation
   slot {class=HorizontalBoxSlot, properties {size {sizeRule=Fill, value=1}}}
  }
  Spacer_54 {
   widgetClassPath=&Spacer
   parent=VerticalBox-DataAssetCreation
   widget=Spacer_54
   properties {size=<1 10>}
   slot {class=VerticalBoxSlot}
  }
  HorizontalBox_4 {
   widgetClassPath=&HorizontalBox
   parent=VerticalBox-DataAssetCreation
   widget=HorizontalBox_4
   slot {class=VerticalBoxSlot, properties {verticalAlignment=VAlign_Center}}
  }
  Spacer_59 {widgetClassPath=&Spacer, parent=HorizontalBox_4, widget=Spacer_59, properties {size=<10 1>}, slot {class=HorizontalBoxSlot}}
  TextBlock_29 {
   widgetClassPath=&TextBlock
   parent=HorizontalBox_4
   widget=TextBlock_29
   properties {
    text="Atom Path: "
    colorAndOpacity {colorUseRule=UseColor_Specified, specifiedColor=<rgba 0.5647 0.5647 0.5647 1>}
    font {
     bForceMonospaced=false
     bMaterialIsStencil=false
     fontMaterial=null
     fontObject=&Roboto
     letterSpacing=0
     monospacedWidth=1
     size=7.5
     skewAmount=0
     typefaceFontName=Regular
     outlineSettings {
      bApplyOutlineToDropShadows=false
      bMiteredCorners=false
      bSeparateFillAlpha=false
      outlineMaterial=null
      outlineSize=0
      outlineColor=<rgba 0 0 0 1>
     }
    }
   }
   slot {class=HorizontalBoxSlot, properties {verticalAlignment=VAlign_Center}}
  }
  Border-DataAssetPath {
   widgetClassPath=&Border
   parent=HorizontalBox_4
   widget=Border-DataAssetPath
   properties {
    toolTipText="Paste here a folder relative Path, e.g. /Game/MyFolder\r\nor the Atom Reference Path --> Select the Atom Model, press CTRL+C and press CTRL+V here.\r\nThe system will search for and display all valid Atom Models within the path provided."
    verticalAlignment=VAlign_Center
    background {
     bIsDynamicallyLoaded=false
     drawAs=RoundedBox
     imageType=NoImage
     mirroring=NoMirror
     resourceName=None
     resourceObject=null
     tiling=NoTile
     imageSize=<32 32>
     margin=<ltrb 0 0 0 0>
     outlineSettings {
      bUseBrushTransparency=false
      roundingType=FixedRadius
      width=1
      color {colorUseRule=UseColor_Specified, specifiedColor=<rgba 0.0625 0.0625 0.0625 1>}
      cornerRadii=<3 3 3 3>
     }
     tintColor {colorUseRule=UseColor_Specified, specifiedColor=<rgba 0.0048 0.0048 0.0048 1>}
     uVRegion {bIsValid=false, max=<0 0>, min=<0 0>}
    }
    contentColorAndOpacity=<rgba 0.5647 0.5647 0.5647 1>
    desiredSizeScale=<1 1.5>
    padding=<ltrb 2 2 2 2>
   }
   slot {class=HorizontalBoxSlot, properties {verticalAlignment=VAlign_Center, size {sizeRule=Fill, value=1}}}
  }
  INP_Path {
   widgetClassPath=/Script/Blutility.EditorUtilityEditableText
   bIsVariable=true
   parent=Border-DataAssetPath
   widget=INP_Path
   properties {
    hintText="Paste here an Atom Model Reference Path or a folder relative path (Required)"
    widgetStyle {
     backgroundImageComposing {
      bIsDynamicallyLoaded=false
      drawAs=Border
      imageType=FullColor
      mirroring=NoMirror
      resourceName=../../../Engine/Content/Slate/Old/HyperlinkDotted.png
      resourceObject=null
      tiling=Both
      imageSize=<0 0>
      margin=<ltrb 0 0 0 0.1875>
      outlineSettings {
       bUseBrushTransparency=false
       roundingType=HalfHeightRadius
       width=0
       color {colorUseRule=UseColor_Specified, specifiedColor=<rgba 0 0 0 0>}
       cornerRadii=<0 0 0 0>
      }
      tintColor {colorUseRule=UseColor_Specified, specifiedColor=<rgba 1 1 1 1>}
      uVRegion {bIsValid=false, max=<0 0>, min=<0 0>}
     }
     backgroundImageSelected {
      bIsDynamicallyLoaded=false
      drawAs=Image
      imageType=NoImage
      mirroring=NoMirror
      resourceName=None
      resourceObject=null
      tiling=NoTile
      imageSize=<0 0>
      margin=<ltrb 0 0 0 0>
      outlineSettings {
       bUseBrushTransparency=false
       roundingType=HalfHeightRadius
       width=0
       color {colorUseRule=UseColor_Specified, specifiedColor=<rgba 0 0 0 0>}
       cornerRadii=<0 0 0 0>
      }
      tintColor {colorUseRule=UseColor_Specified, specifiedColor=<rgba 0 0.162 0.7454 1>}
      uVRegion {bIsValid=false, max=<0 0>, min=<0 0>}
     }
     caretImage {
      bIsDynamicallyLoaded=false
      drawAs=Image
      imageType=NoImage
      mirroring=NoMirror
      resourceName=None
      resourceObject=null
      tiling=NoTile
      imageSize=<0 0>
      margin=<ltrb 0 0 0 0>
      outlineSettings {
       bUseBrushTransparency=false
       roundingType=HalfHeightRadius
       width=0
       color {colorUseRule=UseColor_Specified, specifiedColor=<rgba 0 0 0 0>}
       cornerRadii=<0 0 0 0>
      }
      tintColor {colorUseRule=UseColor_Specified, specifiedColor=<rgba 1 1 1 1>}
      uVRegion {bIsValid=false, max=<0 0>, min=<0 0>}
     }
     colorAndOpacity {colorUseRule=UseColor_Foreground, specifiedColor=<rgba 1 0 1 1>}
     font {
      bForceMonospaced=false
      bMaterialIsStencil=false
      fontMaterial=null
      fontObject=null
      letterSpacing=0
      monospacedWidth=1
      size=7.5
      skewAmount=0
      typefaceFontName=Regular
      outlineSettings {
       bApplyOutlineToDropShadows=false
       bMiteredCorners=false
       bSeparateFillAlpha=false
       outlineMaterial=null
       outlineSize=0
       outlineColor=<rgba 0 0 0 1>
      }
     }
     hintColorAndOpacity {colorUseRule=UseColor_Foreground, specifiedColor=<rgba 1 0 1 1>}
     selectedBackgroundColor {colorUseRule=UseColor_Foreground, specifiedColor=<rgba 1 0 1 1>}
    }
   }
   slot {class=BorderSlot, properties {verticalAlignment=VAlign_Center, padding=<ltrb 2 2 2 2>}}
  }
  Spacer_60 {widgetClassPath=&Spacer, parent=HorizontalBox_4, widget=Spacer_60, properties {size=<10 1>}, slot {class=HorizontalBoxSlot}}
  Spacer_57 {
   widgetClassPath=&Spacer
   parent=VerticalBox-DataAssetCreation
   widget=Spacer_57
   properties {size=<10 10>}
   slot {class=VerticalBoxSlot}
  }
  HorizontalBox_2 {widgetClassPath=&HorizontalBox, parent=VerticalBox-DataAssetCreation, widget=HorizontalBox_2, slot {class=VerticalBoxSlot}}
  Spacer_50 {widgetClassPath=&Spacer, parent=HorizontalBox_2, widget=Spacer_50, properties {size=<1.4 1>}, slot {class=HorizontalBoxSlot}}
  SPV_Proc1 {
   widgetClassPath=&SinglePropertyView
   parent=HorizontalBox_2
   widget=SPV_Proc1
   properties {nameOverride=Processor One, propertyName=Processor_1, toolTipText=Select an Atom processor to run on one batch of Atom models}
   slot {class=HorizontalBoxSlot}
  }
  Spacer_62 {widgetClassPath=&Spacer, parent=HorizontalBox_2, widget=Spacer_62, properties {size=<10 1>}, slot {class=HorizontalBoxSlot}}
  CHB_Custom_Geo_1 {
   widgetClassPath=&EditorUtilityCheckBox
   bIsVariable=true
   parent=HorizontalBox_2
   widget=CHB_Custom_Geo_1
   properties {toolTipText=Check to enable custom geometry settings to run on this processor.}
   slot {class=HorizontalBoxSlot}
  }
  TextBlock_257 {
   widgetClassPath=&TextBlock
   parent=CHB_Custom_Geo_1
   widget=TextBlock_257
   properties {
    text=Custom Geometry
    font {
     bForceMonospaced=false
     bMaterialIsStencil=false
     fontMaterial=null
     fontObject=&Roboto
     letterSpacing=0
     monospacedWidth=1
     size=7.5
     skewAmount=0
     typefaceFontName=Regular
     outlineSettings {
      bApplyOutlineToDropShadows=false
      bMiteredCorners=false
      bSeparateFillAlpha=false
      outlineMaterial=null
      outlineSize=0
      outlineColor=<rgba 0 0 0 1>
     }
    }
   }
   slot {class=PanelSlot}
  }
  HorizontalBox_3 {widgetClassPath=&HorizontalBox, parent=VerticalBox-DataAssetCreation, widget=HorizontalBox_3, slot {class=VerticalBoxSlot}}
  SPV_Proc2 {
   widgetClassPath=&SinglePropertyView
   parent=HorizontalBox_3
   widget=SPV_Proc2
   properties {nameOverride=Processor Two, propertyName=Processor_2, toolTipText=Select an Atom processor to run on one batch of Atom models}
   slot {class=HorizontalBoxSlot}
  }
  Spacer_46 {widgetClassPath=&Spacer, parent=HorizontalBox_3, widget=Spacer_46, properties {size=<10 1>}, slot {class=HorizontalBoxSlot}}
  CHB_Custom_Geo_2 {
   widgetClassPath=&EditorUtilityCheckBox
   bIsVariable=true
   parent=HorizontalBox_3
   widget=CHB_Custom_Geo_2
   properties {toolTipText=Check to enable custom geometry settings to run on this processor.}
   slot {class=HorizontalBoxSlot}
  }
  TextBlock_21 {
   widgetClassPath=&TextBlock
   parent=CHB_Custom_Geo_2
   widget=TextBlock_21
   properties {
    text=Custom Geometry
    font {
     bForceMonospaced=false
     bMaterialIsStencil=false
     fontMaterial=null
     fontObject=&Roboto
     letterSpacing=0
     monospacedWidth=1
     size=7.5
     skewAmount=0
     typefaceFontName=Regular
     outlineSettings {
      bApplyOutlineToDropShadows=false
      bMiteredCorners=false
      bSeparateFillAlpha=false
      outlineMaterial=null
      outlineSize=0
      outlineColor=<rgba 0 0 0 1>
     }
    }
   }
   slot {class=PanelSlot}
  }
  HorizontalBox_365 {widgetClassPath=&HorizontalBox, parent=VerticalBox-DataAssetCreation, widget=HorizontalBox_365, slot {class=VerticalBoxSlot}}
  BTN_SearchAtomModels {
   widgetClassPath=&EditorUtilityButton
   parent=HorizontalBox_365
   widget=BTN_SearchAtomModels
   properties {toolTipText=Creates a new Atom Data Asset with the name and path supplied and populates it automatically.}
   slot {class=HorizontalBoxSlot}
  }
  TextBlock_25 {
   widgetClassPath=&TextBlock
   parent=BTN_SearchAtomModels
   widget=TextBlock_25
   properties {
    text=Search Atom Models
    colorAndOpacity {colorUseRule=UseColor_Specified, specifiedColor=<rgba 0.5647 0.5647 0.5647 1>}
    font {
     bForceMonospaced=false
     bMaterialIsStencil=false
     fontMaterial=null
     fontObject=&Roboto
     letterSpacing=0
     monospacedWidth=1
     size=7.5
     skewAmount=0
     typefaceFontName=Regular
     outlineSettings {
      bApplyOutlineToDropShadows=false
      bMiteredCorners=false
      bSeparateFillAlpha=false
      outlineMaterial=null
      outlineSize=0
      outlineColor=<rgba 0 0 0 1>
     }
    }
   }
   slot {class=ButtonSlot}
  }
  Spacer_51 {widgetClassPath=&Spacer, parent=HorizontalBox_365, widget=Spacer_51, properties {size=<10 1>}, slot {class=HorizontalBoxSlot}}
  CHB_ReimportPrimitives {
   widgetClassPath=&EditorUtilityCheckBox
   bIsVariable=true
   parent=HorizontalBox_365
   widget=CHB_ReimportPrimitives
   properties {toolTipText="Check to have the visualizer reimport all primitives during processing (this may extend processing time considerably.)"}
   slot {class=HorizontalBoxSlot}
  }
  TextBlock_371 {
   widgetClassPath=&TextBlock
   parent=CHB_ReimportPrimitives
   widget=TextBlock_371
   properties {
    text=Reimport Primitives
    font {
     bForceMonospaced=false
     bMaterialIsStencil=false
     fontMaterial=null
     fontObject=&Roboto
     letterSpacing=0
     monospacedWidth=1
     size=7.5
     skewAmount=0
     typefaceFontName=Regular
     outlineSettings {
      bApplyOutlineToDropShadows=false
      bMiteredCorners=false
      bSeparateFillAlpha=false
      outlineMaterial=null
      outlineSize=0
      outlineColor=<rgba 0 0 0 1>
     }
    }
   }
   slot {class=PanelSlot}
  }
  Spacer_21 {
   widgetClassPath=&Spacer
   parent=VerticalBox-DataAssetCreation
   widget=Spacer_21
   properties {size=<1 5>}
   slot {class=VerticalBoxSlot}
  }
  HorizontalBox_1 {
   widgetClassPath=&HorizontalBox
   parent=VerticalBox-DataAssetCreation
   widget=HorizontalBox_1
   slot {class=VerticalBoxSlot, properties {horizontalAlignment=HAlign_Left, verticalAlignment=VAlign_Center, size {sizeRule=Fill, value=1}}}
  }
  Spacer_45 {
   widgetClassPath=&Spacer
   parent=HorizontalBox_1
   widget=Spacer_45
   properties {size=<10 5>}
   slot {class=HorizontalBoxSlot, properties {size {sizeRule=Fill, value=1}}}
  }
  TextBlock {
   widgetClassPath=&TextBlock
   parent=HorizontalBox_1
   widget=TextBlock
   properties {
    text=Displaying:
    colorAndOpacity {colorUseRule=UseColor_Specified, specifiedColor=<rgba 0.5647 0.5647 0.5647 1>}
    font {
     bForceMonospaced=false
     bMaterialIsStencil=false
     fontMaterial=null
     fontObject=&Roboto
     letterSpacing=0
     monospacedWidth=1
     size=7.5
     skewAmount=0
     typefaceFontName=Regular
     outlineSettings {
      bApplyOutlineToDropShadows=false
      bMiteredCorners=false
      bSeparateFillAlpha=false
      outlineMaterial=null
      outlineSize=0
      outlineColor=<rgba 0 0 0 1>
     }
    }
   }
   slot {class=HorizontalBoxSlot}
  }
  Spacer_22 {widgetClassPath=&Spacer, parent=HorizontalBox_1, widget=Spacer_22, properties {size=<5 5>}, slot {class=HorizontalBoxSlot}}
  TXT_CurrentlyLoaded {
   widgetClassPath=&TextBlock
   bIsVariable=true
   parent=HorizontalBox_1
   widget=TXT_CurrentlyLoaded
   properties {
    text=None
    colorAndOpacity {colorUseRule=UseColor_Specified, specifiedColor=<rgba 0.2292 0.2292 0.2292 1>}
    font {
     bForceMonospaced=false
     bMaterialIsStencil=false
     fontMaterial=null
     fontObject=&Roboto
     letterSpacing=0
     monospacedWidth=1
     size=7.5
     skewAmount=0
     typefaceFontName=Regular
     outlineSettings {
      bApplyOutlineToDropShadows=false
      bMiteredCorners=false
      bSeparateFillAlpha=false
      outlineMaterial=null
      outlineSize=0
      outlineColor=<rgba 0 0 0 1>
     }
    }
   }
   slot {class=HorizontalBoxSlot}
  }
  Spacer_14 {widgetClassPath=&Spacer, parent=VerticalBox_0, widget=Spacer_14, properties {size=<0 10>}, slot {class=VerticalBoxSlot}}
  Border_Line_0 {
   widgetClassPath=&Border
   parent=VerticalBox_0
   widget=Border_Line_0
   properties {
    background {
     bIsDynamicallyLoaded=false
     drawAs=Box
     imageType=NoImage
     mirroring=NoMirror
     resourceName=None
     resourceObject=null
     tiling=NoTile
     imageSize=<32 32>
     margin=<ltrb 0 0 0 0>
     outlineSettings {
      bUseBrushTransparency=false
      roundingType=HalfHeightRadius
      width=0
      color {colorUseRule=UseColor_Specified, specifiedColor=<rgba 0 0 0 0>}
      cornerRadii=<0 0 0 0>
     }
     tintColor {colorUseRule=UseColor_Specified, specifiedColor=<rgba 1 1 1 0.25>}
     uVRegion {bIsValid=false, max=<0 0>, min=<0 0>}
    }
    desiredSizeScale=<1 0.3>
    padding=<ltrb 4.1 4.1 4.1 4.1>
   }
   slot {class=VerticalBoxSlot}
  }
  Spacer_23 {widgetClassPath=&Spacer, parent=VerticalBox_0, widget=Spacer_23, properties {size=<1 10>}, slot {class=VerticalBoxSlot}}
  HorizontalBox-ViewModeModel {
   widgetClassPath=&HorizontalBox
   parent=VerticalBox_0
   widget=HorizontalBox-ViewModeModel
   slot {class=VerticalBoxSlot, properties {verticalAlignment=VAlign_Center}}
  }
  Spacer_42 {
   widgetClassPath=&Spacer
   parent=HorizontalBox-ViewModeModel
   widget=Spacer_42
   properties {size=<10 0>}
   slot {class=HorizontalBoxSlot}
  }
  VerticalBox_1 {widgetClassPath=&VerticalBox, parent=HorizontalBox-ViewModeModel, widget=VerticalBox_1, slot {class=HorizontalBoxSlot}}
  TextBlock_1 {
   widgetClassPath=&TextBlock
   parent=VerticalBox_1
   widget=TextBlock_1
   properties {
    text=View Mode
    colorAndOpacity {colorUseRule=UseColor_Specified, specifiedColor=<rgba 0.5647 0.5647 0.5647 1>}
    font {
     bForceMonospaced=false
     bMaterialIsStencil=false
     fontMaterial=null
     fontObject=&Roboto
     letterSpacing=0
     monospacedWidth=1
     size=8.25
     skewAmount=0
     typefaceFontName=Bold
     outlineSettings {
      bApplyOutlineToDropShadows=false
      bMiteredCorners=false
      bSeparateFillAlpha=false
      outlineMaterial=null
      outlineSize=0
      outlineColor=<rgba 0 0 0 1>
     }
    }
   }
   slot {class=VerticalBoxSlot, properties {horizontalAlignment=HAlign_Center}}
  }
  Spacer_2 {widgetClassPath=&Spacer, parent=VerticalBox_1, widget=Spacer_2, properties {size=<1 10>}, slot {class=VerticalBoxSlot}}
  HorizontalBox-ViewMode {widgetClassPath=&HorizontalBox, parent=VerticalBox_1, widget=HorizontalBox-ViewMode, slot {class=VerticalBoxSlot}}
  BTN_GridView {
   widgetClassPath=&EditorUtilityButton
   parent=HorizontalBox-ViewMode
   widget=BTN_GridView
   properties {toolTipText="Change View Mode to 2D Line View, showing all models within the selected Atom Data Asset."}
   slot {class=HorizontalBoxSlot, properties {horizontalAlignment=HAlign_Center, verticalAlignment=VAlign_Center}}
  }
  TextBlock_2 {
   widgetClassPath=&TextBlock
   parent=BTN_GridView
   widget=TextBlock_2
   properties {
    text=Line View
    colorAndOpacity {colorUseRule=UseColor_Specified, specifiedColor=<rgba 0.5647 0.5647 0.5647 1>}
    font {
     bForceMonospaced=false
     bMaterialIsStencil=false
     fontMaterial=null
     fontObject=&Roboto
     letterSpacing=0
     monospacedWidth=1
     size=7.5
     skewAmount=0
     typefaceFontName=Regular
     outlineSettings {
      bApplyOutlineToDropShadows=false
      bMiteredCorners=false
      bSeparateFillAlpha=false
      outlineMaterial=null
      outlineSize=0
      outlineColor=<rgba 0 0 0 1>
     }
    }
   }
   slot {class=ButtonSlot}
  }
  Spacer_1 {widgetClassPath=&Spacer, parent=HorizontalBox-ViewMode, widget=Spacer_1, properties {size=<5 0>}, slot {class=HorizontalBoxSlot}}
  BTN_LODView {
   widgetClassPath=&EditorUtilityButton
   parent=HorizontalBox-ViewMode
   widget=BTN_LODView
   properties {
    toolTipText="Changes the View Mode do LOD View. \r\nDisplays a grid of models, broken down by their LOD."
    widgetStyle {
     clickedSlateSound {resourceObject=null}
     disabled {
      bIsDynamicallyLoaded=false
      drawAs=RoundedBox
      imageType=NoImage
      mirroring=NoMirror
      resourceName=None
      resourceObject=null
      tiling=NoTile
      imageSize=<0 0>
      margin=<ltrb 0 0 0 0>
      outlineSettings {
       bUseBrushTransparency=false
       roundingType=FixedRadius
       width=1
       color {colorUseRule=UseColor_ColorTable, specifiedColor=<rgba 0 0 0 0>}
       cornerRadii=<4 4 4 4>
      }
      tintColor {colorUseRule=UseColor_Specified, specifiedColor=<rgba 0.0395 0.0395 0.0395 1>}
      uVRegion {bIsValid=false, max=<0 0>, min=<0 0>}
     }
     disabledForeground {colorUseRule=UseColor_ColorTable, specifiedColor=<rgba 0 0 0 0>}
     hovered {
      bIsDynamicallyLoaded=false
      drawAs=RoundedBox
      imageType=NoImage
      mirroring=NoMirror
      resourceName=None
      resourceObject=null
      tiling=NoTile
      imageSize=<32 32>
      margin=<ltrb 0 0 0 0>
      outlineSettings {
       bUseBrushTransparency=false
       roundingType=FixedRadius
       width=1
       color {colorUseRule=UseColor_ColorTable, specifiedColor=<rgba 0 0 0 0>}
       cornerRadii=<4 4 4 4>
      }
      tintColor {colorUseRule=UseColor_Specified, specifiedColor=<rgba 0.0953 0.0953 0.0953 1>}
      uVRegion {bIsValid=false, max=<0 0>, min=<0 0>}
     }
     hoveredForeground {colorUseRule=UseColor_ColorTable, specifiedColor=<rgba 0 0 0 0>}
     hoveredSlateSound {resourceObject=null}
     normal {
      bIsDynamicallyLoaded=false
      drawAs=RoundedBox
      imageType=NoImage
      mirroring=NoMirror
      resourceName=None
      resourceObject=null
      tiling=NoTile
      imageSize=<32 32>
      margin=<ltrb 0 0 0 0>
      outlineSettings {
       bUseBrushTransparency=false
       roundingType=FixedRadius
       width=1
       color {colorUseRule=UseColor_ColorTable, specifiedColor=<rgba 0 0 0 0>}
       cornerRadii=<4 4 4 4>
      }
      tintColor {colorUseRule=UseColor_Specified, specifiedColor=<rgba 0.0395 0.0395 0.0395 1>}
      uVRegion {bIsValid=false, max=<0 0>, min=<0 0>}
     }
     normalForeground {colorUseRule=UseColor_ColorTable, specifiedColor=<rgba 0 0 0 0>}
     normalPadding=<ltrb 12 1.5 12 1.5>
     pressed {
      bIsDynamicallyLoaded=false
      drawAs=RoundedBox
      imageType=NoImage
      mirroring=NoMirror
      resourceName=None
      resourceObject=null
      tiling=NoTile
      imageSize=<32 32>
      margin=<ltrb 0 0 0 0>
      outlineSettings {
       bUseBrushTransparency=false
       roundingType=FixedRadius
       width=1
       color {colorUseRule=UseColor_ColorTable, specifiedColor=<rgba 0 0 0 0>}
       cornerRadii=<4 4 4 4>
      }
      tintColor {colorUseRule=UseColor_Specified, specifiedColor=<rgba 0.0284 0.0284 0.0284 1>}
      uVRegion {bIsValid=false, max=<0 0>, min=<0 0>}
     }
     pressedForeground {colorUseRule=UseColor_ColorTable, specifiedColor=<rgba 0 0 0 0>}
     pressedPadding=<ltrb 12 2.5 12 0.5>
     pressedSlateSound {resourceObject=null}
    }
   }
   slot {class=HorizontalBoxSlot, properties {horizontalAlignment=HAlign_Center, verticalAlignment=VAlign_Center}}
  }
  TextBlock_3 {
   widgetClassPath=&TextBlock
   parent=BTN_LODView
   widget=TextBlock_3
   properties {
    text=LOD View
    colorAndOpacity {colorUseRule=UseColor_Specified, specifiedColor=<rgba 0.5647 0.5647 0.5647 1>}
    font {
     bForceMonospaced=false
     bMaterialIsStencil=false
     fontMaterial=null
     fontObject=&Roboto
     letterSpacing=0
     monospacedWidth=1
     size=7.5
     skewAmount=0
     typefaceFontName=Regular
     outlineSettings {
      bApplyOutlineToDropShadows=false
      bMiteredCorners=false
      bSeparateFillAlpha=false
      outlineMaterial=null
      outlineSize=0
      outlineColor=<rgba 0 0 0 1>
     }
    }
   }
   slot {class=ButtonSlot}
  }
  Spacer_36 {
   widgetClassPath=&Spacer
   parent=HorizontalBox-ViewModeModel
   widget=Spacer_36
   properties {size=<10 0>}
   slot {class=HorizontalBoxSlot}
  }
  VerticalBox-DisplayFilter {
   widgetClassPath=&VerticalBox
   parent=HorizontalBox-ViewModeModel
   widget=VerticalBox-DisplayFilter
   slot {class=HorizontalBoxSlot, properties {horizontalAlignment=HAlign_Center, verticalAlignment=VAlign_Center}}
  }
  TextBlock_30 {
   widgetClassPath=&TextBlock
   parent=VerticalBox-DisplayFilter
   widget=TextBlock_30
   properties {
    text=Display Filter
    colorAndOpacity {colorUseRule=UseColor_Specified, specifiedColor=<rgba 0.5647 0.5647 0.5647 1>}
    font {
     bForceMonospaced=false
     bMaterialIsStencil=false
     fontMaterial=null
     fontObject=&Roboto
     letterSpacing=0
     monospacedWidth=1
     size=8.25
     skewAmount=0
     typefaceFontName=Bold
     outlineSettings {
      bApplyOutlineToDropShadows=false
      bMiteredCorners=false
      bSeparateFillAlpha=false
      outlineMaterial=null
      outlineSize=0
      outlineColor=<rgba 0 0 0 1>
     }
    }
   }
   slot {class=VerticalBoxSlot, properties {horizontalAlignment=HAlign_Center}}
  }
  Spacer_70 {
   widgetClassPath=&Spacer
   parent=VerticalBox-DisplayFilter
   widget=Spacer_70
   properties {size=<10 10>}
   slot {class=VerticalBoxSlot, properties {size {sizeRule=Fill, value=1}}}
  }
  HorizontalBox-ViewMode_1 {widgetClassPath=&HorizontalBox, parent=VerticalBox-DisplayFilter, widget=HorizontalBox-ViewMode_1, slot {class=VerticalBoxSlot}}
  CHB_ShowOver {
   widgetClassPath=&EditorUtilityCheckBox
   bIsVariable=true
   parent=HorizontalBox-ViewMode_1
   widget=CHB_ShowOver
   slot {class=HorizontalBoxSlot}
  }
  TextBlock_31 {
   widgetClassPath=&TextBlock
   parent=CHB_ShowOver
   widget=TextBlock_31
   properties {
    text=Show Only Over-Budgeted
    colorAndOpacity {colorUseRule=UseColor_Specified, specifiedColor=<rgba 0.5647 0.5647 0.5647 1>}
    font {
     bForceMonospaced=false
     bMaterialIsStencil=false
     fontMaterial=null
     fontObject=&Roboto
     letterSpacing=0
     monospacedWidth=1
     size=7.5
     skewAmount=0
     typefaceFontName=Regular
     outlineSettings {
      bApplyOutlineToDropShadows=false
      bMiteredCorners=false
      bSeparateFillAlpha=false
      outlineMaterial=null
      outlineSize=0
      outlineColor=<rgba 0 0 0 1>
     }
    }
   }
   slot {class=PanelSlot}
  }
  Spacer_69 {
   widgetClassPath=&Spacer
   parent=HorizontalBox-ViewMode_1
   widget=Spacer_69
   properties {size=<10 0>}
   slot {class=HorizontalBoxSlot}
  }
  Spacer_41 {
   widgetClassPath=&Spacer
   parent=HorizontalBox-ViewModeModel
   widget=Spacer_41
   properties {size=<10 0>}
   slot {class=HorizontalBoxSlot}
  }
  TextBlock_10 {
   widgetClassPath=&TextBlock
   parent=HorizontalBox-ViewModeModel
   widget=TextBlock_10
   properties {
    text=Text Size
    colorAndOpacity {colorUseRule=UseColor_Specified, specifiedColor=<rgba 0.5647 0.5647 0.5647 1>}
    font {
     bForceMonospaced=false
     bMaterialIsStencil=false
     fontMaterial=null
     fontObject=&Roboto
     letterSpacing=0
     monospacedWidth=1
     size=7.5
     skewAmount=0
     typefaceFontName=Regular
     outlineSettings {
      bApplyOutlineToDropShadows=false
      bMiteredCorners=false
      bSeparateFillAlpha=false
      outlineMaterial=null
      outlineSize=0
      outlineColor=<rgba 0 0 0 1>
     }
    }
   }
   slot {class=HorizontalBoxSlot, properties {horizontalAlignment=HAlign_Center, verticalAlignment=VAlign_Center}}
  }
  Border_1 {
   widgetClassPath=&Border
   parent=HorizontalBox-ViewModeModel
   widget=Border_1
   properties {
    horizontalAlignment=HAlign_Center
    verticalAlignment=VAlign_Center
    background {
     bIsDynamicallyLoaded=false
     drawAs=RoundedBox
     imageType=NoImage
     mirroring=NoMirror
     resourceName=None
     resourceObject=null
     tiling=NoTile
     imageSize=<32 32>
     margin=<ltrb 0 0 0 0>
     outlineSettings {
      bUseBrushTransparency=false
      roundingType=FixedRadius
      width=1
      color {colorUseRule=UseColor_Specified, specifiedColor=<rgba 0.0625 0.0625 0.0625 1>}
      cornerRadii=<3 3 3 3>
     }
     tintColor {colorUseRule=UseColor_Specified, specifiedColor=<rgba 0.0048 0.0048 0.0048 1>}
     uVRegion {bIsValid=false, max=<0 0>, min=<0 0>}
    }
    contentColorAndOpacity=<rgba 0.5647 0.5647 0.5647 1>
    padding=<ltrb 7 5 7 5>
   }
   slot {class=HorizontalBoxSlot, properties {horizontalAlignment=HAlign_Center, verticalAlignment=VAlign_Center}}
  }
  INP_TextSize {
   widgetClassPath=&EditorUtilityEditableText
   bIsVariable=true
   parent=Border_1
   widget=INP_TextSize
   properties {
    justification=Center
    text=24.0
    widgetStyle {
     backgroundImageComposing {
      bIsDynamicallyLoaded=false
      drawAs=Border
      imageType=FullColor
      mirroring=NoMirror
      resourceName=../../../Engine/Content/Slate/Old/HyperlinkDotted.png
      resourceObject=null
      tiling=Both
      imageSize=<0 0>
      margin=<ltrb 0 0 0 0.1875>
      outlineSettings {
       bUseBrushTransparency=false
       roundingType=HalfHeightRadius
       width=0
       color {colorUseRule=UseColor_Specified, specifiedColor=<rgba 0 0 0 0>}
       cornerRadii=<0 0 0 0>
      }
      tintColor {colorUseRule=UseColor_Specified, specifiedColor=<rgba 1 1 1 1>}
      uVRegion {bIsValid=false, max=<0 0>, min=<0 0>}
     }
     backgroundImageSelected {
      bIsDynamicallyLoaded=false
      drawAs=Image
      imageType=NoImage
      mirroring=NoMirror
      resourceName=None
      resourceObject=null
      tiling=NoTile
      imageSize=<0 0>
      margin=<ltrb 0 0 0 0>
      outlineSettings {
       bUseBrushTransparency=false
       roundingType=HalfHeightRadius
       width=0
       color {colorUseRule=UseColor_Specified, specifiedColor=<rgba 0 0 0 0>}
       cornerRadii=<0 0 0 0>
      }
      tintColor {colorUseRule=UseColor_Specified, specifiedColor=<rgba 0 0.162 0.7454 1>}
      uVRegion {bIsValid=false, max=<0 0>, min=<0 0>}
     }
     caretImage {
      bIsDynamicallyLoaded=false
      drawAs=Image
      imageType=NoImage
      mirroring=NoMirror
      resourceName=None
      resourceObject=null
      tiling=NoTile
      imageSize=<0 0>
      margin=<ltrb 0 0 0 0>
      outlineSettings {
       bUseBrushTransparency=false
       roundingType=HalfHeightRadius
       width=0
       color {colorUseRule=UseColor_Specified, specifiedColor=<rgba 0 0 0 0>}
       cornerRadii=<0 0 0 0>
      }
      tintColor {colorUseRule=UseColor_Specified, specifiedColor=<rgba 1 1 1 1>}
      uVRegion {bIsValid=false, max=<0 0>, min=<0 0>}
     }
     colorAndOpacity {colorUseRule=UseColor_Foreground, specifiedColor=<rgba 1 0 1 1>}
     font {
      bForceMonospaced=false
      bMaterialIsStencil=false
      fontMaterial=null
      fontObject=null
      letterSpacing=0
      monospacedWidth=1
      size=7.5
      skewAmount=0
      typefaceFontName=Regular
      outlineSettings {
       bApplyOutlineToDropShadows=false
       bMiteredCorners=false
       bSeparateFillAlpha=false
       outlineMaterial=null
       outlineSize=0
       outlineColor=<rgba 0 0 0 1>
      }
     }
     hintColorAndOpacity {colorUseRule=UseColor_Foreground, specifiedColor=<rgba 1 0 1 1>}
     selectedBackgroundColor {colorUseRule=UseColor_Foreground, specifiedColor=<rgba 1 0 1 1>}
    }
   }
   slot {class=BorderSlot, properties {horizontalAlignment=HAlign_Center, verticalAlignment=VAlign_Center, padding=<ltrb 7 5 7 5>}}
  }
  Spacer_40 {
   widgetClassPath=&Spacer
   parent=HorizontalBox-ViewModeModel
   widget=Spacer_40
   properties {size=<10 0>}
   slot {class=HorizontalBoxSlot}
  }
  Spacer_77 {widgetClassPath=&Spacer, parent=VerticalBox_0, widget=Spacer_77, properties {size=<1 10>}, slot {class=VerticalBoxSlot}}
  EditorUtilityExpandableArea_Transform {
   widgetClassPath=/Script/Blutility.EditorUtilityExpandableArea
   bIsVariable=true
   parent=VerticalBox_0
   widget=EditorUtilityExpandableArea_Transform
   properties {
    bIsExpanded=true
    borderBrush {
     bIsDynamicallyLoaded=false
     drawAs=RoundedBox
     imageType=NoImage
     mirroring=NoMirror
     resourceName=None
     resourceObject=null
     tiling=NoTile
     imageSize=<0 0>
     margin=<ltrb 0 0 0 0>
     outlineSettings {
      bUseBrushTransparency=false
      roundingType=FixedRadius
      width=0
      color {colorUseRule=UseColor_Specified, specifiedColor=<rgba 0 0 0 0>}
      cornerRadii=<4 4 4 4>
     }
     tintColor {colorUseRule=UseColor_Specified, specifiedColor=<rgba 0.0284 0.0284 0.0284 1>}
     uVRegion {bIsValid=false, max=<0 0>, min=<0 0>}
    }
   }
   slot {class=VerticalBoxSlot}
  }
  TextBlock_147 {
   widgetClassPath=&TextBlock
   namedSlotHost=EditorUtilityExpandableArea_Transform
   widget=TextBlock_147
   properties {
    text=Transform
    colorAndOpacity {colorUseRule=UseColor_Specified, specifiedColor=<rgba 0.5647 0.5647 0.5647 1>}
    font {
     bForceMonospaced=false
     bMaterialIsStencil=false
     fontMaterial=null
     fontObject=&Roboto
     letterSpacing=0
     monospacedWidth=1
     size=8.25
     skewAmount=0
     typefaceFontName=Bold
     outlineSettings {
      bApplyOutlineToDropShadows=false
      bMiteredCorners=false
      bSeparateFillAlpha=false
      outlineMaterial=null
      outlineSize=0
      outlineColor=<rgba 0 0 0 1>
     }
    }
   }
  }
  VerticalBox-Transform {widgetClassPath=&VerticalBox, namedSlotHost=EditorUtilityExpandableArea_Transform, widget=VerticalBox-Transform}
  Spacer_28 {widgetClassPath=&Spacer, parent=VerticalBox-Transform, widget=Spacer_28, properties {size=<1 10>}, slot {class=VerticalBoxSlot}}
  HorizontalBox-Transform {
   widgetClassPath=&HorizontalBox
   parent=VerticalBox-Transform
   widget=HorizontalBox-Transform
   slot {class=VerticalBoxSlot, properties {horizontalAlignment=HAlign_Left}}
  }
  Spacer_35 {
   widgetClassPath=&Spacer
   parent=HorizontalBox-Transform
   widget=Spacer_35
   properties {size=<10 0>}
   slot {class=HorizontalBoxSlot, properties {size {sizeRule=Fill, value=1}}}
  }
  BTN_RotationClock {
   widgetClassPath=&EditorUtilityButton
   parent=HorizontalBox-Transform
   widget=BTN_RotationClock
   slot {class=HorizontalBoxSlot}
  }
  TextBlock_15 {
   widgetClassPath=&TextBlock
   parent=BTN_RotationClock
   widget=TextBlock_15
   properties {
    text=Rotate Right
    colorAndOpacity {colorUseRule=UseColor_Specified, specifiedColor=<rgba 0.5647 0.5647 0.5647 1>}
    font {
     bForceMonospaced=false
     bMaterialIsStencil=false
     fontMaterial=null
     fontObject=&Roboto
     letterSpacing=0
     monospacedWidth=1
     size=7.5
     skewAmount=0
     typefaceFontName=Regular
     outlineSettings {
      bApplyOutlineToDropShadows=false
      bMiteredCorners=false
      bSeparateFillAlpha=false
      outlineMaterial=null
      outlineSize=0
      outlineColor=<rgba 0 0 0 1>
     }
    }
   }
   slot {class=ButtonSlot}
  }
  Spacer_30 {
   widgetClassPath=&Spacer
   parent=HorizontalBox-Transform
   widget=Spacer_30
   properties {size=<5 0>}
   slot {class=HorizontalBoxSlot, properties {size {sizeRule=Fill, value=1}}}
  }
  BTN_RotationAntiClock {
   widgetClassPath=&EditorUtilityButton
   parent=HorizontalBox-Transform
   widget=BTN_RotationAntiClock
   slot {class=HorizontalBoxSlot}
  }
  TextBlock_16 {
   widgetClassPath=&TextBlock
   parent=BTN_RotationAntiClock
   widget=TextBlock_16
   properties {
    text=Rotate Left
    colorAndOpacity {colorUseRule=UseColor_Specified, specifiedColor=<rgba 0.5647 0.5647 0.5647 1>}
    font {
     bForceMonospaced=false
     bMaterialIsStencil=false
     fontMaterial=null
     fontObject=&Roboto
     letterSpacing=0
     monospacedWidth=1
     size=7.5
     skewAmount=0
     typefaceFontName=Regular
     outlineSettings {
      bApplyOutlineToDropShadows=false
      bMiteredCorners=false
      bSeparateFillAlpha=false
      outlineMaterial=null
      outlineSize=0
      outlineColor=<rgba 0 0 0 1>
     }
    }
   }
   slot {class=ButtonSlot}
  }
  Spacer_31 {
   widgetClassPath=&Spacer
   parent=HorizontalBox-Transform
   widget=Spacer_31
   properties {size=<5 0>}
   slot {class=HorizontalBoxSlot, properties {size {sizeRule=Fill, value=1}}}
  }
  BTN_RotationReset {
   widgetClassPath=&EditorUtilityButton
   parent=HorizontalBox-Transform
   widget=BTN_RotationReset
   slot {class=HorizontalBoxSlot}
  }
  TextBlock_17 {
   widgetClassPath=&TextBlock
   parent=BTN_RotationReset
   widget=TextBlock_17
   properties {
    text=Reset Rotation
    colorAndOpacity {colorUseRule=UseColor_Specified, specifiedColor=<rgba 0.5647 0.5647 0.5647 1>}
    font {
     bForceMonospaced=false
     bMaterialIsStencil=false
     fontMaterial=null
     fontObject=&Roboto
     letterSpacing=0
     monospacedWidth=1
     size=7.5
     skewAmount=0
     typefaceFontName=Regular
     outlineSettings {
      bApplyOutlineToDropShadows=false
      bMiteredCorners=false
      bSeparateFillAlpha=false
      outlineMaterial=null
      outlineSize=0
      outlineColor=<rgba 0 0 0 1>
     }
    }
   }
   slot {class=ButtonSlot}
  }
  Spacer_34 {
   widgetClassPath=&Spacer
   parent=HorizontalBox-Transform
   widget=Spacer_34
   properties {size=<10 0>}
   slot {class=HorizontalBoxSlot, properties {size {sizeRule=Fill, value=1}}}
  }
  Spacer_32 {widgetClassPath=&Spacer, parent=VerticalBox-Transform, widget=Spacer_32, properties {size=<1 15>}, slot {class=VerticalBoxSlot}}
  HorizontalBox-HeightOffset {
   widgetClassPath=&HorizontalBox
   parent=VerticalBox-Transform
   widget=HorizontalBox-HeightOffset
   slot {class=VerticalBoxSlot, properties {horizontalAlignment=HAlign_Left}}
  }
  Spacer {
   widgetClassPath=&Spacer
   parent=HorizontalBox-HeightOffset
   widget=Spacer
   properties {size=<10 0>}
   slot {class=HorizontalBoxSlot, properties {size {sizeRule=Fill, value=1}}}
  }
  TextBlock_19 {
   widgetClassPath=&TextBlock
   parent=HorizontalBox-HeightOffset
   widget=TextBlock_19
   properties {
    text=Height Offset
    colorAndOpacity {colorUseRule=UseColor_Specified, specifiedColor=<rgba 0.5647 0.5647 0.5647 1>}
    font {
     bForceMonospaced=false
     bMaterialIsStencil=false
     fontMaterial=null
     fontObject=&Roboto
     letterSpacing=0
     monospacedWidth=1
     size=7.5
     skewAmount=0
     typefaceFontName=Regular
     outlineSettings {
      bApplyOutlineToDropShadows=false
      bMiteredCorners=false
      bSeparateFillAlpha=false
      outlineMaterial=null
      outlineSize=0
      outlineColor=<rgba 0 0 0 1>
     }
    }
   }
   slot {class=HorizontalBoxSlot, properties {horizontalAlignment=HAlign_Center, verticalAlignment=VAlign_Center}}
  }
  Spacer_38 {
   widgetClassPath=&Spacer
   parent=HorizontalBox-HeightOffset
   widget=Spacer_38
   properties {size=<5 0>}
   slot {class=HorizontalBoxSlot}
  }
  Border_2 {
   widgetClassPath=&Border
   parent=HorizontalBox-HeightOffset
   widget=Border_2
   properties {
    horizontalAlignment=HAlign_Center
    verticalAlignment=VAlign_Center
    background {
     bIsDynamicallyLoaded=false
     drawAs=RoundedBox
     imageType=NoImage
     mirroring=NoMirror
     resourceName=None
     resourceObject=null
     tiling=NoTile
     imageSize=<32 32>
     margin=<ltrb 0 0 0 0>
     outlineSettings {
      bUseBrushTransparency=false
      roundingType=FixedRadius
      width=1
      color {colorUseRule=UseColor_Specified, specifiedColor=<rgba 0.0625 0.0625 0.0625 1>}
      cornerRadii=<3 3 3 3>
     }
     tintColor {colorUseRule=UseColor_Specified, specifiedColor=<rgba 0.0048 0.0048 0.0048 1>}
     uVRegion {bIsValid=false, max=<0 0>, min=<0 0>}
    }
    contentColorAndOpacity=<rgba 0.5647 0.5647 0.5647 1>
    padding=<ltrb 7 5 7 5>
   }
   slot {class=HorizontalBoxSlot, properties {horizontalAlignment=HAlign_Center, verticalAlignment=VAlign_Center}}
  }
  INP_HeightOffset {
   widgetClassPath=&EditorUtilityEditableText
   bIsVariable=true
   parent=Border_2
   widget=INP_HeightOffset
   properties {
    text=0
    widgetStyle {
     backgroundImageComposing {
      bIsDynamicallyLoaded=false
      drawAs=Border
      imageType=FullColor
      mirroring=NoMirror
      resourceName=../../../Engine/Content/Slate/Old/HyperlinkDotted.png
      resourceObject=null
      tiling=Both
      imageSize=<0 0>
      margin=<ltrb 0 0 0 0.1875>
      outlineSettings {
       bUseBrushTransparency=false
       roundingType=HalfHeightRadius
       width=0
       color {colorUseRule=UseColor_Specified, specifiedColor=<rgba 0 0 0 0>}
       cornerRadii=<0 0 0 0>
      }
      tintColor {colorUseRule=UseColor_Specified, specifiedColor=<rgba 1 1 1 1>}
      uVRegion {bIsValid=false, max=<0 0>, min=<0 0>}
     }
     backgroundImageSelected {
      bIsDynamicallyLoaded=false
      drawAs=Image
      imageType=NoImage
      mirroring=NoMirror
      resourceName=None
      resourceObject=null
      tiling=NoTile
      imageSize=<0 0>
      margin=<ltrb 0 0 0 0>
      outlineSettings {
       bUseBrushTransparency=false
       roundingType=HalfHeightRadius
       width=0
       color {colorUseRule=UseColor_Specified, specifiedColor=<rgba 0 0 0 0>}
       cornerRadii=<0 0 0 0>
      }
      tintColor {colorUseRule=UseColor_Specified, specifiedColor=<rgba 0 0.162 0.7454 1>}
      uVRegion {bIsValid=false, max=<0 0>, min=<0 0>}
     }
     caretImage {
      bIsDynamicallyLoaded=false
      drawAs=Image
      imageType=NoImage
      mirroring=NoMirror
      resourceName=None
      resourceObject=null
      tiling=NoTile
      imageSize=<0 0>
      margin=<ltrb 0 0 0 0>
      outlineSettings {
       bUseBrushTransparency=false
       roundingType=HalfHeightRadius
       width=0
       color {colorUseRule=UseColor_Specified, specifiedColor=<rgba 0 0 0 0>}
       cornerRadii=<0 0 0 0>
      }
      tintColor {colorUseRule=UseColor_Specified, specifiedColor=<rgba 1 1 1 1>}
      uVRegion {bIsValid=false, max=<0 0>, min=<0 0>}
     }
     colorAndOpacity {colorUseRule=UseColor_Foreground, specifiedColor=<rgba 1 0 1 1>}
     font {
      bForceMonospaced=false
      bMaterialIsStencil=false
      fontMaterial=null
      fontObject=null
      letterSpacing=0
      monospacedWidth=1
      size=7.5
      skewAmount=0
      typefaceFontName=Regular
      outlineSettings {
       bApplyOutlineToDropShadows=false
       bMiteredCorners=false
       bSeparateFillAlpha=false
       outlineMaterial=null
       outlineSize=0
       outlineColor=<rgba 0 0 0 1>
      }
     }
     hintColorAndOpacity {colorUseRule=UseColor_Foreground, specifiedColor=<rgba 1 0 1 1>}
     selectedBackgroundColor {colorUseRule=UseColor_Foreground, specifiedColor=<rgba 1 0 1 1>}
    }
   }
   slot {class=BorderSlot, properties {horizontalAlignment=HAlign_Center, verticalAlignment=VAlign_Center, padding=<ltrb 7 5 7 5>}}
  }
  Spacer_37 {
   widgetClassPath=&Spacer
   parent=HorizontalBox-HeightOffset
   widget=Spacer_37
   properties {size=<15 0>}
   slot {class=HorizontalBoxSlot, properties {size {sizeRule=Fill, value=1}}}
  }
  TextBlock_18 {
   widgetClassPath=&TextBlock
   parent=HorizontalBox-HeightOffset
   widget=TextBlock_18
   properties {
    text=Rotation Increment
    colorAndOpacity {colorUseRule=UseColor_Specified, specifiedColor=<rgba 0.5647 0.5647 0.5647 1>}
    font {
     bForceMonospaced=false
     bMaterialIsStencil=false
     fontMaterial=null
     fontObject=&Roboto
     letterSpacing=0
     monospacedWidth=1
     size=7.5
     skewAmount=0
     typefaceFontName=Regular
     outlineSettings {
      bApplyOutlineToDropShadows=false
      bMiteredCorners=false
      bSeparateFillAlpha=false
      outlineMaterial=null
      outlineSize=0
      outlineColor=<rgba 0 0 0 1>
     }
    }
   }
   slot {class=HorizontalBoxSlot, properties {horizontalAlignment=HAlign_Center, verticalAlignment=VAlign_Center}}
  }
  Spacer_39 {
   widgetClassPath=&Spacer
   parent=HorizontalBox-HeightOffset
   widget=Spacer_39
   properties {size=<5 0>}
   slot {class=HorizontalBoxSlot}
  }
  Border {
   widgetClassPath=&Border
   parent=HorizontalBox-HeightOffset
   widget=Border
   properties {
    horizontalAlignment=HAlign_Center
    verticalAlignment=VAlign_Center
    background {
     bIsDynamicallyLoaded=false
     drawAs=RoundedBox
     imageType=NoImage
     mirroring=NoMirror
     resourceName=None
     resourceObject=null
     tiling=NoTile
     imageSize=<32 32>
     margin=<ltrb 0 0 0 0>
     outlineSettings {
      bUseBrushTransparency=false
      roundingType=FixedRadius
      width=1
      color {colorUseRule=UseColor_Specified, specifiedColor=<rgba 0.0625 0.0625 0.0625 1>}
      cornerRadii=<3 3 3 3>
     }
     tintColor {colorUseRule=UseColor_Specified, specifiedColor=<rgba 0.0048 0.0048 0.0048 1>}
     uVRegion {bIsValid=false, max=<0 0>, min=<0 0>}
    }
    contentColorAndOpacity=<rgba 0.5647 0.5647 0.5647 1>
    padding=<ltrb 7 5 7 5>
   }
   slot {class=HorizontalBoxSlot, properties {horizontalAlignment=HAlign_Center, verticalAlignment=VAlign_Center}}
  }
  INP_RotationInc {
   widgetClassPath=&EditorUtilityEditableText
   bIsVariable=true
   parent=Border
   widget=INP_RotationInc
   properties {
    justification=Center
    text=45
    widgetStyle {
     backgroundImageComposing {
      bIsDynamicallyLoaded=false
      drawAs=Border
      imageType=FullColor
      mirroring=NoMirror
      resourceName=../../../Engine/Content/Slate/Old/HyperlinkDotted.png
      resourceObject=null
      tiling=Both
      imageSize=<0 0>
      margin=<ltrb 0 0 0 0.1875>
      outlineSettings {
       bUseBrushTransparency=false
       roundingType=HalfHeightRadius
       width=0
       color {colorUseRule=UseColor_Specified, specifiedColor=<rgba 0 0 0 0>}
       cornerRadii=<0 0 0 0>
      }
      tintColor {colorUseRule=UseColor_Specified, specifiedColor=<rgba 1 1 1 1>}
      uVRegion {bIsValid=false, max=<0 0>, min=<0 0>}
     }
     backgroundImageSelected {
      bIsDynamicallyLoaded=false
      drawAs=Image
      imageType=NoImage
      mirroring=NoMirror
      resourceName=None
      resourceObject=null
      tiling=NoTile
      imageSize=<0 0>
      margin=<ltrb 0 0 0 0>
      outlineSettings {
       bUseBrushTransparency=false
       roundingType=HalfHeightRadius
       width=0
       color {colorUseRule=UseColor_Specified, specifiedColor=<rgba 0 0 0 0>}
       cornerRadii=<0 0 0 0>
      }
      tintColor {colorUseRule=UseColor_Specified, specifiedColor=<rgba 0 0.162 0.7454 1>}
      uVRegion {bIsValid=false, max=<0 0>, min=<0 0>}
     }
     caretImage {
      bIsDynamicallyLoaded=false
      drawAs=Image
      imageType=NoImage
      mirroring=NoMirror
      resourceName=None
      resourceObject=null
      tiling=NoTile
      imageSize=<0 0>
      margin=<ltrb 0 0 0 0>
      outlineSettings {
       bUseBrushTransparency=false
       roundingType=HalfHeightRadius
       width=0
       color {colorUseRule=UseColor_Specified, specifiedColor=<rgba 0 0 0 0>}
       cornerRadii=<0 0 0 0>
      }
      tintColor {colorUseRule=UseColor_Specified, specifiedColor=<rgba 1 1 1 1>}
      uVRegion {bIsValid=false, max=<0 0>, min=<0 0>}
     }
     colorAndOpacity {colorUseRule=UseColor_Foreground, specifiedColor=<rgba 1 0 1 1>}
     font {
      bForceMonospaced=false
      bMaterialIsStencil=false
      fontMaterial=null
      fontObject=null
      letterSpacing=0
      monospacedWidth=1
      size=7.5
      skewAmount=0
      typefaceFontName=Regular
      outlineSettings {
       bApplyOutlineToDropShadows=false
       bMiteredCorners=false
       bSeparateFillAlpha=false
       outlineMaterial=null
       outlineSize=0
       outlineColor=<rgba 0 0 0 1>
      }
     }
     hintColorAndOpacity {colorUseRule=UseColor_Foreground, specifiedColor=<rgba 1 0 1 1>}
     selectedBackgroundColor {colorUseRule=UseColor_Foreground, specifiedColor=<rgba 1 0 1 1>}
    }
   }
   slot {class=BorderSlot, properties {horizontalAlignment=HAlign_Center, verticalAlignment=VAlign_Center, padding=<ltrb 7 5 7 5>}}
  }
  Spacer_27 {
   widgetClassPath=&Spacer
   parent=HorizontalBox-HeightOffset
   widget=Spacer_27
   properties {size=<10 0>}
   slot {class=HorizontalBoxSlot, properties {size {sizeRule=Fill, value=1}}}
  }
  TextBlock_22 {
   widgetClassPath=&TextBlock
   parent=HorizontalBox-HeightOffset
   widget=TextBlock_22
   properties {
    text=Column Spacing
    colorAndOpacity {colorUseRule=UseColor_Specified, specifiedColor=<rgba 0.5647 0.5647 0.5647 1>}
    font {
     bForceMonospaced=false
     bMaterialIsStencil=false
     fontMaterial=null
     fontObject=&Roboto
     letterSpacing=0
     monospacedWidth=1
     size=7.5
     skewAmount=0
     typefaceFontName=Regular
     outlineSettings {
      bApplyOutlineToDropShadows=false
      bMiteredCorners=false
      bSeparateFillAlpha=false
      outlineMaterial=null
      outlineSize=0
      outlineColor=<rgba 0 0 0 1>
     }
    }
   }
   slot {class=HorizontalBoxSlot, properties {horizontalAlignment=HAlign_Center, verticalAlignment=VAlign_Center}}
  }
  Spacer_53 {
   widgetClassPath=&Spacer
   parent=HorizontalBox-HeightOffset
   widget=Spacer_53
   properties {size=<5 0>}
   slot {class=HorizontalBoxSlot}
  }
  Border_3 {
   widgetClassPath=&Border
   parent=HorizontalBox-HeightOffset
   widget=Border_3
   properties {
    horizontalAlignment=HAlign_Center
    verticalAlignment=VAlign_Center
    background {
     bIsDynamicallyLoaded=false
     drawAs=RoundedBox
     imageType=NoImage
     mirroring=NoMirror
     resourceName=None
     resourceObject=null
     tiling=NoTile
     imageSize=<32 32>
     margin=<ltrb 0 0 0 0>
     outlineSettings {
      bUseBrushTransparency=false
      roundingType=FixedRadius
      width=1
      color {colorUseRule=UseColor_Specified, specifiedColor=<rgba 0.0625 0.0625 0.0625 1>}
      cornerRadii=<3 3 3 3>
     }
     tintColor {colorUseRule=UseColor_Specified, specifiedColor=<rgba 0.0048 0.0048 0.0048 1>}
     uVRegion {bIsValid=false, max=<0 0>, min=<0 0>}
    }
    contentColorAndOpacity=<rgba 0.5647 0.5647 0.5647 1>
    padding=<ltrb 7 5 7 5>
   }
   slot {class=HorizontalBoxSlot, properties {horizontalAlignment=HAlign_Center, verticalAlignment=VAlign_Center}}
  }
  INP_RowSpacing {
   widgetClassPath=&EditorUtilityEditableText
   bIsVariable=true
   parent=Border_3
   widget=INP_RowSpacing
   properties {
    justification=Center
    text=2.0
    widgetStyle {
     backgroundImageComposing {
      bIsDynamicallyLoaded=false
      drawAs=Border
      imageType=FullColor
      mirroring=NoMirror
      resourceName=../../../Engine/Content/Slate/Old/HyperlinkDotted.png
      resourceObject=null
      tiling=Both
      imageSize=<0 0>
      margin=<ltrb 0 0 0 0.1875>
      outlineSettings {
       bUseBrushTransparency=false
       roundingType=HalfHeightRadius
       width=0
       color {colorUseRule=UseColor_Specified, specifiedColor=<rgba 0 0 0 0>}
       cornerRadii=<0 0 0 0>
      }
      tintColor {colorUseRule=UseColor_Specified, specifiedColor=<rgba 1 1 1 1>}
      uVRegion {bIsValid=false, max=<0 0>, min=<0 0>}
     }
     backgroundImageSelected {
      bIsDynamicallyLoaded=false
      drawAs=Image
      imageType=NoImage
      mirroring=NoMirror
      resourceName=None
      resourceObject=null
      tiling=NoTile
      imageSize=<0 0>
      margin=<ltrb 0 0 0 0>
      outlineSettings {
       bUseBrushTransparency=false
       roundingType=HalfHeightRadius
       width=0
       color {colorUseRule=UseColor_Specified, specifiedColor=<rgba 0 0 0 0>}
       cornerRadii=<0 0 0 0>
      }
      tintColor {colorUseRule=UseColor_Specified, specifiedColor=<rgba 0 0.162 0.7454 1>}
      uVRegion {bIsValid=false, max=<0 0>, min=<0 0>}
     }
     caretImage {
      bIsDynamicallyLoaded=false
      drawAs=Image
      imageType=NoImage
      mirroring=NoMirror
      resourceName=None
      resourceObject=null
      tiling=NoTile
      imageSize=<0 0>
      margin=<ltrb 0 0 0 0>
      outlineSettings {
       bUseBrushTransparency=false
       roundingType=HalfHeightRadius
       width=0
       color {colorUseRule=UseColor_Specified, specifiedColor=<rgba 0 0 0 0>}
       cornerRadii=<0 0 0 0>
      }
      tintColor {colorUseRule=UseColor_Specified, specifiedColor=<rgba 1 1 1 1>}
      uVRegion {bIsValid=false, max=<0 0>, min=<0 0>}
     }
     colorAndOpacity {colorUseRule=UseColor_Foreground, specifiedColor=<rgba 1 0 1 1>}
     font {
      bForceMonospaced=false
      bMaterialIsStencil=false
      fontMaterial=null
      fontObject=null
      letterSpacing=0
      monospacedWidth=1
      size=7.5
      skewAmount=0
      typefaceFontName=Regular
      outlineSettings {
       bApplyOutlineToDropShadows=false
       bMiteredCorners=false
       bSeparateFillAlpha=false
       outlineMaterial=null
       outlineSize=0
       outlineColor=<rgba 0 0 0 1>
      }
     }
     hintColorAndOpacity {colorUseRule=UseColor_Foreground, specifiedColor=<rgba 1 0 1 1>}
     selectedBackgroundColor {colorUseRule=UseColor_Foreground, specifiedColor=<rgba 1 0 1 1>}
    }
   }
   slot {class=BorderSlot, properties {horizontalAlignment=HAlign_Center, verticalAlignment=VAlign_Center, padding=<ltrb 7 5 7 5>}}
  }
  Spacer_29 {widgetClassPath=&Spacer, parent=VerticalBox-Transform, widget=Spacer_29, properties {size=<1 5>}, slot {class=VerticalBoxSlot}}
  Spacer_47 {widgetClassPath=&Spacer, parent=VerticalBox_0, widget=Spacer_47, properties {size=<1 10>}, slot {class=VerticalBoxSlot}}
  Border_Line_3 {
   widgetClassPath=&Border
   parent=VerticalBox_0
   widget=Border_Line_3
   properties {
    background {
     bIsDynamicallyLoaded=false
     drawAs=Box
     imageType=NoImage
     mirroring=NoMirror
     resourceName=None
     resourceObject=null
     tiling=NoTile
     imageSize=<32 32>
     margin=<ltrb 0 0 0 0>
     outlineSettings {
      bUseBrushTransparency=false
      roundingType=HalfHeightRadius
      width=0
      color {colorUseRule=UseColor_Specified, specifiedColor=<rgba 0 0 0 0>}
      cornerRadii=<0 0 0 0>
     }
     tintColor {colorUseRule=UseColor_Specified, specifiedColor=<rgba 1 1 1 0.25>}
     uVRegion {bIsValid=false, max=<0 0>, min=<0 0>}
    }
    desiredSizeScale=<1 0.3>
    padding=<ltrb 4.1 4.1 4.1 4.1>
   }
   slot {class=VerticalBoxSlot}
  }
  VerticalBox-LODSetup {
   widgetClassPath=&VerticalBox
   parent=VerticalBox_0
   widget=VerticalBox-LODSetup
   slot {class=VerticalBoxSlot, properties {horizontalAlignment=HAlign_Left}}
  }
  Spacer_6 {widgetClassPath=&Spacer, parent=VerticalBox-LODSetup, widget=Spacer_6, properties {size=<1 10>}, slot {class=VerticalBoxSlot}}
  TextBlock_8 {
   widgetClassPath=&TextBlock
   parent=VerticalBox-LODSetup
   widget=TextBlock_8
   properties {
    text=Nanite and LOD Setup
    colorAndOpacity {colorUseRule=UseColor_Specified, specifiedColor=<rgba 0.5647 0.5647 0.5647 1>}
    font {
     bForceMonospaced=false
     bMaterialIsStencil=false
     fontMaterial=null
     fontObject=&Roboto
     letterSpacing=0
     monospacedWidth=1
     size=8.25
     skewAmount=0
     typefaceFontName=Bold
     outlineSettings {
      bApplyOutlineToDropShadows=false
      bMiteredCorners=false
      bSeparateFillAlpha=false
      outlineMaterial=null
      outlineSize=0
      outlineColor=<rgba 0 0 0 1>
     }
    }
   }
   slot {class=VerticalBoxSlot, properties {horizontalAlignment=HAlign_Center}}
  }
  Spacer_7 {widgetClassPath=&Spacer, parent=VerticalBox-LODSetup, widget=Spacer_7, properties {size=<1 10>}, slot {class=VerticalBoxSlot}}
  HorizontalBox-NaniteLODs {widgetClassPath=&HorizontalBox, parent=VerticalBox-LODSetup, widget=HorizontalBox-NaniteLODs, slot {class=VerticalBoxSlot}}
  Spacer_17 {
   widgetClassPath=&Spacer
   parent=HorizontalBox-NaniteLODs
   widget=Spacer_17
   properties {size=<10 0>}
   slot {class=HorizontalBoxSlot, properties {size {sizeRule=Fill, value=1}}}
  }
  BTN_DisableNanite {
   widgetClassPath=&EditorUtilityButton
   parent=HorizontalBox-NaniteLODs
   widget=BTN_DisableNanite
   properties {toolTipText="Whether to Disable Nanite Globally, displaying LODs instead.\r\nr.nanite 0"}
   slot {class=HorizontalBoxSlot}
  }
  TextBlock_4 {
   widgetClassPath=&TextBlock
   parent=BTN_DisableNanite
   widget=TextBlock_4
   properties {
    text=Disable Nanite/LOD Auto
    colorAndOpacity {colorUseRule=UseColor_Specified, specifiedColor=<rgba 0.5647 0.5647 0.5647 1>}
    font {
     bForceMonospaced=false
     bMaterialIsStencil=false
     fontMaterial=null
     fontObject=&Roboto
     letterSpacing=0
     monospacedWidth=1
     size=7.5
     skewAmount=0
     typefaceFontName=Regular
     outlineSettings {
      bApplyOutlineToDropShadows=false
      bMiteredCorners=false
      bSeparateFillAlpha=false
      outlineMaterial=null
      outlineSize=0
      outlineColor=<rgba 0 0 0 1>
     }
    }
   }
   slot {class=ButtonSlot}
  }
  Spacer_5 {
   widgetClassPath=&Spacer
   parent=HorizontalBox-NaniteLODs
   widget=Spacer_5
   properties {size=<5 0>}
   slot {class=HorizontalBoxSlot}
  }
  BTN_EnableNanite {
   widgetClassPath=&EditorUtilityButton
   parent=HorizontalBox-NaniteLODs
   widget=BTN_EnableNanite
   properties {toolTipText="Whether to Enable Nanite, globally.\r\nr.Nanite 1"}
   slot {class=HorizontalBoxSlot}
  }
  TextBlock_5 {
   widgetClassPath=&TextBlock
   parent=BTN_EnableNanite
   widget=TextBlock_5
   properties {
    text=Enable Nanite
    colorAndOpacity {colorUseRule=UseColor_Specified, specifiedColor=<rgba 0.5647 0.5647 0.5647 1>}
    font {
     bForceMonospaced=false
     bMaterialIsStencil=false
     fontMaterial=null
     fontObject=&Roboto
     letterSpacing=0
     monospacedWidth=1
     size=7.5
     skewAmount=0
     typefaceFontName=Regular
     outlineSettings {
      bApplyOutlineToDropShadows=false
      bMiteredCorners=false
      bSeparateFillAlpha=false
      outlineMaterial=null
      outlineSize=0
      outlineColor=<rgba 0 0 0 1>
     }
    }
   }
   slot {class=ButtonSlot}
  }
  Spacer_4 {
   widgetClassPath=&Spacer
   parent=HorizontalBox-NaniteLODs
   widget=Spacer_4
   properties {size=<5 0>}
   slot {class=HorizontalBoxSlot}
  }
  BTN_LODLower {
   widgetClassPath=&EditorUtilityButton
   parent=HorizontalBox-NaniteLODs
   widget=BTN_LODLower
   properties {toolTipText=Forces all models to show their previous LOD}
   slot {class=HorizontalBoxSlot}
  }
  TextBlock_6 {
   widgetClassPath=&TextBlock
   parent=BTN_LODLower
   widget=TextBlock_6
   properties {
    text=Lower LOD
    colorAndOpacity {colorUseRule=UseColor_Specified, specifiedColor=<rgba 0.5647 0.5647 0.5647 1>}
    font {
     bForceMonospaced=false
     bMaterialIsStencil=false
     fontMaterial=null
     fontObject=&Roboto
     letterSpacing=0
     monospacedWidth=1
     size=7.5
     skewAmount=0
     typefaceFontName=Regular
     outlineSettings {
      bApplyOutlineToDropShadows=false
      bMiteredCorners=false
      bSeparateFillAlpha=false
      outlineMaterial=null
      outlineSize=0
      outlineColor=<rgba 0 0 0 1>
     }
    }
   }
   slot {class=ButtonSlot}
  }
  Spacer_8 {
   widgetClassPath=&Spacer
   parent=HorizontalBox-NaniteLODs
   widget=Spacer_8
   properties {size=<5 0>}
   slot {class=HorizontalBoxSlot}
  }
  BTN_LODHigher {
   widgetClassPath=&EditorUtilityButton
   parent=HorizontalBox-NaniteLODs
   widget=BTN_LODHigher
   properties {toolTipText=Forces all models to show their subsequent LOD}
   slot {class=HorizontalBoxSlot}
  }
  TextBlock_7 {
   widgetClassPath=&TextBlock
   parent=BTN_LODHigher
   widget=TextBlock_7
   properties {
    text=Higher LOD
    colorAndOpacity {colorUseRule=UseColor_Specified, specifiedColor=<rgba 0.5647 0.5647 0.5647 1>}
    font {
     bForceMonospaced=false
     bMaterialIsStencil=false
     fontMaterial=null
     fontObject=&Roboto
     letterSpacing=0
     monospacedWidth=1
     size=7.5
     skewAmount=0
     typefaceFontName=Regular
     outlineSettings {
      bApplyOutlineToDropShadows=false
      bMiteredCorners=false
      bSeparateFillAlpha=false
      outlineMaterial=null
      outlineSize=0
      outlineColor=<rgba 0 0 0 1>
     }
    }
   }
   slot {class=ButtonSlot}
  }
  Spacer_18 {
   widgetClassPath=&Spacer
   parent=HorizontalBox-NaniteLODs
   widget=Spacer_18
   properties {size=<10 0>}
   slot {class=HorizontalBoxSlot, properties {size {sizeRule=Fill, value=1}}}
  }
  Spacer_43 {widgetClassPath=&Spacer, parent=VerticalBox_0, widget=Spacer_43, properties {size=<1 10>}, slot {class=VerticalBoxSlot}}
  Border_Line_1 {
   widgetClassPath=&Border
   parent=VerticalBox_0
   widget=Border_Line_1
   properties {
    background {
     bIsDynamicallyLoaded=false
     drawAs=Box
     imageType=NoImage
     mirroring=NoMirror
     resourceName=None
     resourceObject=null
     tiling=NoTile
     imageSize=<32 32>
     margin=<ltrb 0 0 0 0>
     outlineSettings {
      bUseBrushTransparency=false
      roundingType=HalfHeightRadius
      width=0
      color {colorUseRule=UseColor_Specified, specifiedColor=<rgba 0 0 0 0>}
      cornerRadii=<0 0 0 0>
     }
     tintColor {colorUseRule=UseColor_Specified, specifiedColor=<rgba 1 1 1 0.25>}
     uVRegion {bIsValid=false, max=<0 0>, min=<0 0>}
    }
    desiredSizeScale=<1 0.3>
    padding=<ltrb 4.1 4.1 4.1 4.1>
   }
   slot {class=VerticalBoxSlot}
  }
  Spacer_44 {widgetClassPath=&Spacer, parent=VerticalBox_0, widget=Spacer_44, properties {size=<1 10>}, slot {class=VerticalBoxSlot}}
  DetailsView_AssetsPreview {
   widgetClassPath=&DetailsView
   parent=VerticalBox_0
   widget=DetailsView_AssetsPreview
   properties {toolTipText=Auto Populated variables. Displaying the current Atom Model and its main referencers, categoriesToShow [AtomModelList]}
   slot {class=VerticalBoxSlot}
  }
  Spacer_24 {widgetClassPath=&Spacer, parent=VerticalBox_0, widget=Spacer_24, properties {size=<1 10>}, slot {class=VerticalBoxSlot}}
  HorizontalBox-AtomModelButtons {widgetClassPath=&HorizontalBox, parent=VerticalBox_0, widget=HorizontalBox-AtomModelButtons, slot {class=VerticalBoxSlot}}
  Spacer_3 {
   widgetClassPath=&Spacer
   parent=HorizontalBox-AtomModelButtons
   widget=Spacer_3
   properties {size=<10 0>}
   slot {class=HorizontalBoxSlot}
  }
  BTN_ProcessAtomModel {
   widgetClassPath=&EditorUtilityButton
   parent=HorizontalBox-AtomModelButtons
   widget=BTN_ProcessAtomModel
   properties {toolTipText=Process the currently selected Atom Model.}
   slot {class=HorizontalBoxSlot}
  }
  TextBlock_20 {
   widgetClassPath=&TextBlock
   parent=BTN_ProcessAtomModel
   widget=TextBlock_20
   properties {
    text=Process Selected Atom Model
    colorAndOpacity {colorUseRule=UseColor_Specified, specifiedColor=<rgba 0.5647 0.5647 0.5647 1>}
    font {
     bForceMonospaced=false
     bMaterialIsStencil=false
     fontMaterial=null
     fontObject=&Roboto
     letterSpacing=0
     monospacedWidth=1
     size=7.5
     skewAmount=0
     typefaceFontName=Regular
     outlineSettings {
      bApplyOutlineToDropShadows=false
      bMiteredCorners=false
      bSeparateFillAlpha=false
      outlineMaterial=null
      outlineSize=0
      outlineColor=<rgba 0 0 0 1>
     }
    }
   }
   slot {class=ButtonSlot}
  }
  Spacer_49 {
   widgetClassPath=&Spacer
   parent=HorizontalBox-AtomModelButtons
   widget=Spacer_49
   properties {size=<10 0>}
   slot {class=HorizontalBoxSlot}
  }
  BTN_ProcessAtomModels {
   widgetClassPath=&EditorUtilityButton
   parent=HorizontalBox-AtomModelButtons
   widget=BTN_ProcessAtomModels
   properties {toolTipText=Process all Atom Models in the selected Atom Data Asset.}
   slot {class=HorizontalBoxSlot}
  }
  TextBlock_9 {
   widgetClassPath=&TextBlock
   parent=BTN_ProcessAtomModels
   widget=TextBlock_9
   properties {
    text=Process All Atom Models
    colorAndOpacity {colorUseRule=UseColor_Specified, specifiedColor=<rgba 0.5647 0.5647 0.5647 1>}
    font {
     bForceMonospaced=false
     bMaterialIsStencil=false
     fontMaterial=null
     fontObject=&Roboto
     letterSpacing=0
     monospacedWidth=1
     size=7.5
     skewAmount=0
     typefaceFontName=Regular
     outlineSettings {
      bApplyOutlineToDropShadows=false
      bMiteredCorners=false
      bSeparateFillAlpha=false
      outlineMaterial=null
      outlineSize=0
      outlineColor=<rgba 0 0 0 1>
     }
    }
   }
   slot {class=ButtonSlot}
  }
  Spacer_52 {
   widgetClassPath=&Spacer
   parent=HorizontalBox-AtomModelButtons
   widget=Spacer_52
   properties {size=<10 0>}
   slot {class=HorizontalBoxSlot}
  }
  Spacer_48 {widgetClassPath=&Spacer, parent=VerticalBox_0, widget=Spacer_48, properties {size=<1 10>}, slot {class=VerticalBoxSlot}}
  Border_Line_2 {
   widgetClassPath=&Border
   parent=VerticalBox_0
   widget=Border_Line_2
   properties {
    background {
     bIsDynamicallyLoaded=false
     drawAs=Box
     imageType=NoImage
     mirroring=NoMirror
     resourceName=None
     resourceObject=null
     tiling=NoTile
     imageSize=<32 32>
     margin=<ltrb 0 0 0 0>
     outlineSettings {
      bUseBrushTransparency=false
      roundingType=HalfHeightRadius
      width=0
      color {colorUseRule=UseColor_Specified, specifiedColor=<rgba 0 0 0 0>}
      cornerRadii=<0 0 0 0>
     }
     tintColor {colorUseRule=UseColor_Specified, specifiedColor=<rgba 1 1 1 0.25>}
     uVRegion {bIsValid=false, max=<0 0>, min=<0 0>}
    }
    desiredSizeScale=<1 0.3>
    padding=<ltrb 4.1 4.1 4.1 4.1>
   }
   slot {class=VerticalBoxSlot}
  }
 ]
}
