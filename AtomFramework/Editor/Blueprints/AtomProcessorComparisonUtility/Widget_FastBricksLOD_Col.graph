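# Editor comparison widget that shows per-LOD triangle counts for a model's base and FB (likely FastBricks) versions side by side.
# A grid holds a LOD label, base count and FB count for LOD 0 through 6, under a model name header.
# UpdateTriCounts takes the name and two integer arrays and writes each indexed entry to its text block.
# Array lengths are not checked and the event graph is empty.
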WidgetBlueprint /AtomFramework/Editor/Blueprints/AtomProcessorComparisonUtility/Widget_FastBricksLOD_Col
graphs [
 EventGraph {links [], nodes []}
 UpdateTriCounts {nodes [
   n1 = UpdateTriCounts!() -> (then, Model Name:String, BaseTriCount:Array of Integers, FBTriCount:Array of Integers) {is=title K2Node_FunctionEntry}
   > n2 = Sequence()
   > Set Text(Text="To Text (String)"(InString=@n1), self=Get ModelName() -> (ModelName:Text Object Reference))
   n3 = Sequence!(execute=@n2.then_1) -> (then_0, then_1, then_2:Exec, then_3:Exec, then_4:Exec, then_5:Exec, then_6:Exec, then_7:Exec)
   n4 = Sequence!(execute=@n3.then_7) -> (then_0, then_1, then_2, then_3, then_4, then_5, then_6)
   Set Text(
    execute=@n4.then_6
    Text="To Text (Integer)"(Value="Get (a copy)"(Array:Array of Integers=@n1.FBTriCount, Dimension 1=6) -> (Output:int))
    self=Get FBLOD6Text() -> (FBLOD6Text:Text Object Reference)
   )
   Set Text(
    execute=@n4.then_0
    Text="To Text (Integer)"(Value="Get (a copy)"(Array=@n1.FBTriCount))
    self=Get FBLOD0Text() -> (FBLOD0Text:Text Object Reference)
   )
   Set Text(
    execute=@n4.then_1
    Text="To Text (Integer)"(Value="Get (a copy)"(Array=@n1.FBTriCount, Dimension 1=1))
    self=Get FBLOD1Text() -> (FBLOD1Text:Text Object Reference)
   )
   Set Text(
    execute=@n4.then_2
    Text="To Text (Integer)"(Value="Get (a copy)"(Array=@n1.FBTriCount, Dimension 1=2))
    self=Get FBLOD2Text() -> (FBLOD2Text:Text Object Reference)
   )
   Set Text(
    execute=@n4.then_3
    Text="To Text (Integer)"(Value="Get (a copy)"(Array=@n1.FBTriCount, Dimension 1=3))
    self=Get FBLOD3Text() -> (FBLOD3Text:Text Object Reference)
   )
   Set Text(
    execute=@n4.then_4
    Text="To Text (Integer)"(Value="Get (a copy)"(Array=@n1.FBTriCount, Dimension 1=4))
    self=Get FBLOD4Text() -> (FBLOD4Text:Text Object Reference)
   )
   Set Text(
    execute=@n4.then_5
    Text="To Text (Integer)"(Value="Get (a copy)"(Array=@n1.FBTriCount, Dimension 1=5))
    self=Get FBLOD5Text() -> (FBLOD5Text:Text Object Reference)
   )
   Set Text(
    execute=@n3.then_4
    Text="To Text (Integer)"(Value="Get (a copy)"(Array=@n1.BaseTriCount, Dimension 1=4))
    self=Get BaseLOD4Text() -> (BaseLOD4Text:Text Object Reference)
   )
   Set Text(
    execute=@n3.then_5
    Text="To Text (Integer)"(Value="Get (a copy)"(Array=@n1.BaseTriCount, Dimension 1=5))
    self=Get BaseLOD5Text() -> (BaseLOD5Text:Text Object Reference)
   )
   Set Text(
    execute=@n3.then_6
    Text="To Text (Integer)"(Value="Get (a copy)"(Array=@n1.BaseTriCount, Dimension 1=6))
    self=Get BaseLOD6Text() -> (BaseLOD6Text:Text Object Reference)
   )
   Set Text(
    execute=@n3.then_0
    Text="To Text (Integer)"(Value="Get (a copy)"(Array=@n1.BaseTriCount))
    self=Get BaseLOD0Text() -> (BaseLOD0Text:Text Object Reference)
   )
   Set Text(
    execute=@n3.then_1
    Text="To Text (Integer)"(Value="Get (a copy)"(Array=@n1.BaseTriCount, Dimension 1=1))
    self=Get BaseLOD1Text() -> (BaseLOD1Text:Text Object Reference)
   )
   Set Text(
    execute=@n3.then_2
    Text="To Text (Integer)"(Value="Get (a copy)"(Array=@n1.BaseTriCount, Dimension 1=2))
    self=Get BaseLOD2Text() -> (BaseLOD2Text:Text Object Reference)
   )
   Set Text(
    execute=@n3.then_3
    Text="To Text (Integer)"(Value="Get (a copy)"(Array=@n1.BaseTriCount, Dimension 1=3))
    self=Get BaseLOD3Text() -> (BaseLOD3Text:Text Object Reference)
   )
  ]}
]
variables []
widgets {
 info {inheritedWidgetCount=0, parentClass=&UserWidget, rootWidgetClass=&VerticalBox, widgetCount=25}
 widgets [
  VerticalBox_80 {widgetClassPath=&VerticalBox, widget=VerticalBox_80}
  Border_94 {
   widgetClassPath=&Border
   parent=VerticalBox_80
   widget=Border_94
   properties {horizontalAlignment=HAlign_Center, verticalAlignment=VAlign_Center, brushColor=<rgba 1 1 1 0>, padding=<ltrb 5 5 5 5>}
   slot {class=VerticalBoxSlot, properties {padding=<ltrb 5 5 5 5>}}
  }
  ModelName {
   widgetClassPath=&TextBlock
   bIsVariable=true
   parent=Border_94
   widget=ModelName
   properties {text=Text Block}
   slot {class=BorderSlot, properties {horizontalAlignment=HAlign_Center, verticalAlignment=VAlign_Center, padding=<ltrb 5 5 5 5>}}
  }
  GridPanel_138 {
   widgetClassPath=&GridPanel
   parent=VerticalBox_80
   widget=GridPanel_138
   slot {class=VerticalBoxSlot, properties {padding=<ltrb 5 5 5 5>, size {sizeRule=Fill, value=1}}}
  }
  BaseLOD0Text {
   widgetClassPath=&TextBlock
   bIsVariable=true
   parent=GridPanel_138
   widget=BaseLOD0Text
   properties {
    text=LOD 0
    font {
     bForceMonospaced=false
     bMaterialIsStencil=false
     fontMaterial=null
     fontObject=&Roboto
     letterSpacing=0
     monospacedWidth=1
     size=48
     skewAmount=0
     typefaceFontName=Bold
     outlineSettings {
      bApplyOutlineToDropShadows=false
      bMiteredCorners=false
      bSeparateFillAlpha=false
      outlineMaterial=null
      outlineSize=0
      outlineColor=<rgba 0 0 0 1>
     }
    }
   }
   slot {class=GridSlot, properties {column=1, horizontalAlignment=HAlign_Center, verticalAlignment=VAlign_Center, padding=<ltrb 50 5 50 5>}}
  }
  BaseLOD1Text {
   widgetClassPath=&TextBlock
   bIsVariable=true
   parent=GridPanel_138
   widget=BaseLOD1Text
   properties {
    text=LOD 1
    font {
     bForceMonospaced=false
     bMaterialIsStencil=false
     fontMaterial=null
     fontObject=&Roboto
     letterSpacing=0
     monospacedWidth=1
     size=48
     skewAmount=0
     typefaceFontName=Bold
     outlineSettings {
      bApplyOutlineToDropShadows=false
      bMiteredCorners=false
      bSeparateFillAlpha=false
      outlineMaterial=null
      outlineSize=0
      outlineColor=<rgba 0 0 0 1>
     }
    }
   }
   slot {class=GridSlot, properties {column=1, horizontalAlignment=HAlign_Center, row=1, verticalAlignment=VAlign_Center, padding=<ltrb 50 5 50 5>}}
  }
  BaseLOD2Text {
   widgetClassPath=&TextBlock
   bIsVariable=true
   parent=GridPanel_138
   widget=BaseLOD2Text
   properties {
    text=LOD 2
    font {
     bForceMonospaced=false
     bMaterialIsStencil=false
     fontMaterial=null
     fontObject=&Roboto
     letterSpacing=0
     monospacedWidth=1
     size=48
     skewAmount=0
     typefaceFontName=Bold
     outlineSettings {
      bApplyOutlineToDropShadows=false
      bMiteredCorners=false
      bSeparateFillAlpha=false
      outlineMaterial=null
      outlineSize=0
      outlineColor=<rgba 0 0 0 1>
     }
    }
   }
   slot {class=GridSlot, properties {column=1, horizontalAlignment=HAlign_Center, row=2, verticalAlignment=VAlign_Center, padding=<ltrb 50 5 50 5>}}
  }
  BaseLOD3Text {
   widgetClassPath=&TextBlock
   bIsVariable=true
   parent=GridPanel_138
   widget=BaseLOD3Text
   properties {
    text=LOD 3
    font {
     bForceMonospaced=false
     bMaterialIsStencil=false
     fontMaterial=null
     fontObject=&Roboto
     letterSpacing=0
     monospacedWidth=1
     size=48
     skewAmount=0
     typefaceFontName=Bold
     outlineSettings {
      bApplyOutlineToDropShadows=false
      bMiteredCorners=false
      bSeparateFillAlpha=false
      outlineMaterial=null
      outlineSize=0
      outlineColor=<rgba 0 0 0 1>
     }
    }
   }
   slot {class=GridSlot, properties {column=1, horizontalAlignment=HAlign_Center, row=3, verticalAlignment=VAlign_Center, padding=<ltrb 50 5 50 5>}}
  }
  BaseLOD4Text {
   widgetClassPath=&TextBlock
   bIsVariable=true
   parent=GridPanel_138
   widget=BaseLOD4Text
   properties {
    text=LOD 4
    font {
     bForceMonospaced=false
     bMaterialIsStencil=false
     fontMaterial=null
     fontObject=&Roboto
     letterSpacing=0
     monospacedWidth=1
     size=48
     skewAmount=0
     typefaceFontName=Bold
     outlineSettings {
      bApplyOutlineToDropShadows=false
      bMiteredCorners=false
      bSeparateFillAlpha=false
      outlineMaterial=null
      outlineSize=0
      outlineColor=<rgba 0 0 0 1>
     }
    }
   }
   slot {class=GridSlot, properties {column=1, horizontalAlignment=HAlign_Center, row=4, verticalAlignment=VAlign_Center, padding=<ltrb 50 5 50 5>}}
  }
  BaseLOD5Text {
   widgetClassPath=&TextBlock
   bIsVariable=true
   parent=GridPanel_138
   widget=BaseLOD5Text
   properties {
    text=LOD 5
    font {
     bForceMonospaced=false
     bMaterialIsStencil=false
     fontMaterial=null
     fontObject=&Roboto
     letterSpacing=0
     monospacedWidth=1
     size=48
     skewAmount=0
     typefaceFontName=Bold
     outlineSettings {
      bApplyOutlineToDropShadows=false
      bMiteredCorners=false
      bSeparateFillAlpha=false
      outlineMaterial=null
      outlineSize=0
      outlineColor=<rgba 0 0 0 1>
     }
    }
   }
   slot {class=GridSlot, properties {column=1, horizontalAlignment=HAlign_Center, row=5, verticalAlignment=VAlign_Center, padding=<ltrb 50 5 50 5>}}
  }
  BaseLOD6Text {
   widgetClassPath=&TextBlock
   bIsVariable=true
   parent=GridPanel_138
   widget=BaseLOD6Text
   properties {
    text=LOD 6
    font {
     bForceMonospaced=false
     bMaterialIsStencil=false
     fontMaterial=null
     fontObject=&Roboto
     letterSpacing=0
     monospacedWidth=1
     size=48
     skewAmount=0
     typefaceFontName=Bold
     outlineSettings {
      bApplyOutlineToDropShadows=false
      bMiteredCorners=false
      bSeparateFillAlpha=false
      outlineMaterial=null
      outlineSize=0
      outlineColor=<rgba 0 0 0 1>
     }
    }
   }
   slot {class=GridSlot, properties {column=1, horizontalAlignment=HAlign_Center, row=6, verticalAlignment=VAlign_Center, padding=<ltrb 50 5 50 5>}}
  }
  FBLOD0Text {
   widgetClassPath=&TextBlock
   bIsVariable=true
   parent=GridPanel_138
   widget=FBLOD0Text
   properties {
    text=LOD 0
    font {
     bForceMonospaced=false
     bMaterialIsStencil=false
     fontMaterial=null
     fontObject=&Roboto
     letterSpacing=0
     monospacedWidth=1
     size=48
     skewAmount=0
     typefaceFontName=Bold
     outlineSettings {
      bApplyOutlineToDropShadows=false
      bMiteredCorners=false
      bSeparateFillAlpha=false
      outlineMaterial=null
      outlineSize=0
      outlineColor=<rgba 0 0 0 1>
     }
    }
   }
   slot {class=GridSlot, properties {column=2, horizontalAlignment=HAlign_Center, verticalAlignment=VAlign_Center, padding=<ltrb 50 5 50 5>}}
  }
  FBLOD1Text {
   widgetClassPath=&TextBlock
   bIsVariable=true
   parent=GridPanel_138
   widget=FBLOD1Text
   properties {
    text=LOD 1
    font {
     bForceMonospaced=false
     bMaterialIsStencil=false
     fontMaterial=null
     fontObject=&Roboto
     letterSpacing=0
     monospacedWidth=1
     size=48
     skewAmount=0
     typefaceFontName=Bold
     outlineSettings {
      bApplyOutlineToDropShadows=false
      bMiteredCorners=false
      bSeparateFillAlpha=false
      outlineMaterial=null
      outlineSize=0
      outlineColor=<rgba 0 0 0 1>
     }
    }
   }
   slot {class=GridSlot, properties {column=2, horizontalAlignment=HAlign_Center, row=1, verticalAlignment=VAlign_Center, padding=<ltrb 50 5 50 5>}}
  }
  FBLOD2Text {
   widgetClassPath=&TextBlock
   bIsVariable=true
   parent=GridPanel_138
   widget=FBLOD2Text
   properties {
    text=LOD 2
    font {
     bForceMonospaced=false
     bMaterialIsStencil=false
     fontMaterial=null
     fontObject=&Roboto
     letterSpacing=0
     monospacedWidth=1
     size=48
     skewAmount=0
     typefaceFontName=Bold
     outlineSettings {
      bApplyOutlineToDropShadows=false
      bMiteredCorners=false
      bSeparateFillAlpha=false
      outlineMaterial=null
      outlineSize=0
      outlineColor=<rgba 0 0 0 1>
     }
    }
   }
   slot {class=GridSlot, properties {column=2, horizontalAlignment=HAlign_Center, row=2, verticalAlignment=VAlign_Center, padding=<ltrb 50 5 50 5>}}
  }
  FBLOD3Text {
   widgetClassPath=&TextBlock
   bIsVariable=true
   parent=GridPanel_138
   widget=FBLOD3Text
   properties {
    text=LOD 3
    font {
     bForceMonospaced=false
     bMaterialIsStencil=false
     fontMaterial=null
     fontObject=&Roboto
     letterSpacing=0
     monospacedWidth=1
     size=48
     skewAmount=0
     typefaceFontName=Bold
     outlineSettings {
      bApplyOutlineToDropShadows=false
      bMiteredCorners=false
      bSeparateFillAlpha=false
      outlineMaterial=null
      outlineSize=0
      outlineColor=<rgba 0 0 0 1>
     }
    }
   }
   slot {class=GridSlot, properties {column=2, horizontalAlignment=HAlign_Center, row=3, verticalAlignment=VAlign_Center, padding=<ltrb 50 5 50 5>}}
  }
  FBLOD4Text {
   widgetClassPath=&TextBlock
   bIsVariable=true
   parent=GridPanel_138
   widget=FBLOD4Text
   properties {
    text=LOD 4
    font {
     bForceMonospaced=false
     bMaterialIsStencil=false
     fontMaterial=null
     fontObject=&Roboto
     letterSpacing=0
     monospacedWidth=1
     size=48
     skewAmount=0
     typefaceFontName=Bold
     outlineSettings {
      bApplyOutlineToDropShadows=false
      bMiteredCorners=false
      bSeparateFillAlpha=false
      outlineMaterial=null
      outlineSize=0
      outlineColor=<rgba 0 0 0 1>
     }
    }
   }
   slot {class=GridSlot, properties {column=2, horizontalAlignment=HAlign_Center, row=4, verticalAlignment=VAlign_Center, padding=<ltrb 50 5 50 5>}}
  }
  FBLOD5Text {
   widgetClassPath=&TextBlock
   bIsVariable=true
   parent=GridPanel_138
   widget=FBLOD5Text
   properties {
    text=LOD 5
    font {
     bForceMonospaced=false
     bMaterialIsStencil=false
     fontMaterial=null
     fontObject=&Roboto
     letterSpacing=0
     monospacedWidth=1
     size=48
     skewAmount=0
     typefaceFontName=Bold
     outlineSettings {
      bApplyOutlineToDropShadows=false
      bMiteredCorners=false
      bSeparateFillAlpha=false
      outlineMaterial=null
      outlineSize=0
      outlineColor=<rgba 0 0 0 1>
     }
    }
   }
   slot {class=GridSlot, properties {column=2, horizontalAlignment=HAlign_Center, row=5, verticalAlignment=VAlign_Center, padding=<ltrb 50 5 50 5>}}
  }
  FBLOD6Text {
   widgetClassPath=&TextBlock
   bIsVariable=true
   parent=GridPanel_138
   widget=FBLOD6Text
   properties {
    text=LOD 6
    font {
     bForceMonospaced=false
     bMaterialIsStencil=false
     fontMaterial=null
     fontObject=&Roboto
     letterSpacing=0
     monospacedWidth=1
     size=48
     skewAmount=0
     typefaceFontName=Bold
     outlineSettings {
      bApplyOutlineToDropShadows=false
      bMiteredCorners=false
      bSeparateFillAlpha=false
      outlineMaterial=null
      outlineSize=0
      outlineColor=<rgba 0 0 0 1>
     }
    }
   }
   slot {class=GridSlot, properties {column=2, horizontalAlignment=HAlign_Center, row=6, verticalAlignment=VAlign_Center, padding=<ltrb 50 5 50 5>}}
  }
  LOD0Label {
   widgetClassPath=&TextBlock
   parent=GridPanel_138
   widget=LOD0Label
   properties {
    text=LOD 0
    font {
     bForceMonospaced=false
     bMaterialIsStencil=false
     fontMaterial=null
     fontObject=&Roboto
     letterSpacing=0
     monospacedWidth=1
     size=48
     skewAmount=0
     typefaceFontName=Bold
     outlineSettings {
      bApplyOutlineToDropShadows=false
      bMiteredCorners=false
      bSeparateFillAlpha=false
      outlineMaterial=null
      outlineSize=0
      outlineColor=<rgba 0 0 0 1>
     }
    }
   }
   slot {class=GridSlot, properties {horizontalAlignment=HAlign_Center, verticalAlignment=VAlign_Center, padding=<ltrb 50 5 50 5>}}
  }
  LOD1Label {
   widgetClassPath=&TextBlock
   parent=GridPanel_138
   widget=LOD1Label
   properties {
    text=LOD 1
    font {
     bForceMonospaced=false
     bMaterialIsStencil=false
     fontMaterial=null
     fontObject=&Roboto
     letterSpacing=0
     monospacedWidth=1
     size=48
     skewAmount=0
     typefaceFontName=Bold
     outlineSettings {
      bApplyOutlineToDropShadows=false
      bMiteredCorners=false
      bSeparateFillAlpha=false
      outlineMaterial=null
      outlineSize=0
      outlineColor=<rgba 0 0 0 1>
     }
    }
   }
   slot {class=GridSlot, properties {horizontalAlignment=HAlign_Center, row=1, verticalAlignment=VAlign_Center, padding=<ltrb 50 5 50 5>}}
  }
  LOD2Label {
   widgetClassPath=&TextBlock
   parent=GridPanel_138
   widget=LOD2Label
   properties {
    text=LOD 2
    font {
     bForceMonospaced=false
     bMaterialIsStencil=false
     fontMaterial=null
     fontObject=&Roboto
     letterSpacing=0
     monospacedWidth=1
     size=48
     skewAmount=0
     typefaceFontName=Bold
     outlineSettings {
      bApplyOutlineToDropShadows=false
      bMiteredCorners=false
      bSeparateFillAlpha=false
      outlineMaterial=null
      outlineSize=0
      outlineColor=<rgba 0 0 0 1>
     }
    }
   }
   slot {class=GridSlot, properties {horizontalAlignment=HAlign_Center, row=2, verticalAlignment=VAlign_Center, padding=<ltrb 50 5 50 5>}}
  }
  LOD3Label {
   widgetClassPath=&TextBlock
   parent=GridPanel_138
   widget=LOD3Label
   properties {
    text=LOD 3
    font {
     bForceMonospaced=false
     bMaterialIsStencil=false
     fontMaterial=null
     fontObject=&Roboto
     letterSpacing=0
     monospacedWidth=1
     size=48
     skewAmount=0
     typefaceFontName=Bold
     outlineSettings {
      bApplyOutlineToDropShadows=false
      bMiteredCorners=false
      bSeparateFillAlpha=false
      outlineMaterial=null
      outlineSize=0
      outlineColor=<rgba 0 0 0 1>
     }
    }
   }
   slot {class=GridSlot, properties {horizontalAlignment=HAlign_Center, row=3, verticalAlignment=VAlign_Center, padding=<ltrb 50 5 50 5>}}
  }
  LOD4Label {
   widgetClassPath=&TextBlock
   parent=GridPanel_138
   widget=LOD4Label
   properties {
    text=LOD 4
    font {
     bForceMonospaced=false
     bMaterialIsStencil=false
     fontMaterial=null
     fontObject=&Roboto
     letterSpacing=0
     monospacedWidth=1
     size=48
     skewAmount=0
     typefaceFontName=Bold
     outlineSettings {
      bApplyOutlineToDropShadows=false
      bMiteredCorners=false
      bSeparateFillAlpha=false
      outlineMaterial=null
      outlineSize=0
      outlineColor=<rgba 0 0 0 1>
     }
    }
   }
   slot {class=GridSlot, properties {horizontalAlignment=HAlign_Center, row=4, verticalAlignment=VAlign_Center, padding=<ltrb 50 5 50 5>}}
  }
  LOD5Label {
   widgetClassPath=&TextBlock
   parent=GridPanel_138
   widget=LOD5Label
   properties {
    text=LOD 5
    font {
     bForceMonospaced=false
     bMaterialIsStencil=false
     fontMaterial=null
     fontObject=&Roboto
     letterSpacing=0
     monospacedWidth=1
     size=48
     skewAmount=0
     typefaceFontName=Bold
     outlineSettings {
      bApplyOutlineToDropShadows=false
      bMiteredCorners=false
      bSeparateFillAlpha=false
      outlineMaterial=null
      outlineSize=0
      outlineColor=<rgba 0 0 0 1>
     }
    }
   }
   slot {class=GridSlot, properties {horizontalAlignment=HAlign_Center, row=5, verticalAlignment=VAlign_Center, padding=<ltrb 50 5 50 5>}}
  }
  LOD6Label {
   widgetClassPath=&TextBlock
   parent=GridPanel_138
   widget=LOD6Label
   properties {
    text=LOD 6
    font {
     bForceMonospaced=false
     bMaterialIsStencil=false
     fontMaterial=null
     fontObject=&Roboto
     letterSpacing=0
     monospacedWidth=1
     size=48
     skewAmount=0
     typefaceFontName=Bold
     outlineSettings {
      bApplyOutlineToDropShadows=false
      bMiteredCorners=false
      bSeparateFillAlpha=false
      outlineMaterial=null
      outlineSize=0
      outlineColor=<rgba 0 0 0 1>
     }
    }
   }
   slot {class=GridSlot, properties {horizontalAlignment=HAlign_Center, row=6, verticalAlignment=VAlign_Center, padding=<ltrb 50 5 50 5>}}
  }
 ]
}
