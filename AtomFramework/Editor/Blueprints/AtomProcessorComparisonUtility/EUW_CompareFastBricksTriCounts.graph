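# Editor utility widget that compares base and FastBricks triangle counts for Atom models.
# Clicking Generate Tri Counts fetches a list of per-model count structs and passes it to AddDataGridToEUW.
# That event creates one Widget_FastBricksLOD_Col per model, adds it as a column in a horizontal scroll box, and fills it with the model name and both count arrays.
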EditorUtilityWidgetBlueprint /AtomFramework/Editor/Blueprints/AtomProcessorComparisonUtility/EUW_CompareFastBricksTriCounts
graphs [EventGraph {nodes [
   "On Clicked (EditorUtilityButton_52)"() {is=title K2Node_ComponentBoundEvent}
   > n1 = GetModelFastBricksTriCounts!(execute, __WorldContext:Object Reference) -> (then, Model Tri Counts List:Array of STRUCT Atom Fast Bricks Model Tri Count Structures)
   > AddDataGridToEUW!(execute, self, Model Tri Count List:Array of STRUCT Atom Fast Bricks Model Tri Count Structures=@n1) -> (then)
   n2 = AddDataGridToEUW!() -> (OutputDelegate, then, Model Tri Count List:Array of STRUCT Atom Fast Bricks Model Tri Count Structures) {is=title K2Node_CustomEvent}
   > n3 = For Each Loop(Array:Array of STRUCT Atom Fast Bricks Model Tri Count Structures=@n2.Model Tri Count List) -> (Array Element:STRUCT Atom Fast Bricks Model Tri Count Structure)
   > n4 = Create Widget!(
    execute:Exec
    WorldContextObject:Object Reference=Self-Reference()
    Class:User Widget Class Reference=&/Widget_FastBricksLOD_Col.Widget_FastBricksLOD_Col_C
    OwningPlayer:Player Controller Object Reference
   ) -> (then:Exec, ReturnValue:Widget Fast Bricks LOD Col Object Reference)
   > AddChild(self=Get EditorUtilityScrollBox_121() -> (EditorUtilityScrollBox_121:Editor Utility Scroll Box Object Reference), Content=@n4)
   n5 = Break STRUCT Atom Fast Bricks Model Tri Count!(STRUCT_AtomFastBricksModelTriCount:"STRUCT Atom Fast Bricks Model Tri Count Structure (by ref)"=@n3) -> (
    ModelName_2_#1:String
    BaseTriCount_9_#2:Array of Integers
    FastBricksTriCount_10_#3:Array of Integers
   ) {is=title K2Node_BreakStruct}
   > UpdateTriCounts!(
    execute
    self:Widget Fast Bricks LOD Col Object Reference=@n4
    Model Name:String=@n5
    BaseTriCount:Array of Integers=@n5.BaseTriCount_9_#2
    FBTriCount:Array of Integers=@n5.FastBricksTriCount_10_#3
   ) -> (then)
  ]}]
variables []
widgets {
 info {rootWidgetClass=&CanvasPanel, widgetCount=4}
 widgets [
  CanvasPanel_46 {widgetClassPath=&CanvasPanel, widget=CanvasPanel_46}
  EditorUtilityButton_52 {
   widgetClassPath=&EditorUtilityButton
   parent=CanvasPanel_46
   widget=EditorUtilityButton_52
   slot {
    class=CanvasPanelSlot
    properties {bAutoSize=true, layoutData {alignment=<0 0>, anchors {maximum=<0.5 0>, minimum=<0.5 0>}, offsets=<ltrb -136.4005 12 262.557 76.7351>}}
   }
  }
  TextBlock_97 {
   widgetClassPath=&TextBlock
   parent=EditorUtilityButton_52
   widget=TextBlock_97
   properties {text=Generate Tri Counts}
   slot {class=ButtonSlot}
  }
  EditorUtilityScrollBox_121 {
   widgetClassPath=&EditorUtilityScrollBox
   bIsVariable=true
   parent=CanvasPanel_46
   widget=EditorUtilityScrollBox_121
   properties {orientation=Orient_Horizontal}
   slot {
    class=CanvasPanelSlot
    properties {layoutData {alignment=<0 0>, anchors {maximum=<0.5 0.5>, minimum=<0.5 0.5>}, offsets=<ltrb -534.6652 -352.9973 1075.342 895.6161>}}
   }
  }
 ]
}
